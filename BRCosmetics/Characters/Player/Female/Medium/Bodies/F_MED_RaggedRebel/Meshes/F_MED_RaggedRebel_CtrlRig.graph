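# Control Rig for the F_MED_RaggedRebel player body cosmetic that drives deformation and accessory bones from body motion.
# After the shared DeformRig function, weighted parent constraints pin straps, bags, bullets and scarf to the body surface.
# Swing-twist rotation deltas, spherical pose readers and aim constraints then correct the thigh strap, shoe tongues, neck support, bullets, right knee and chest radio.
# A debug flag, off by default, lets controls pose leg bones for testing.
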ControlRigBlueprint /BRCosmetics/Characters/Player/Female/Medium/Bodies/F_MED_RaggedRebel/Meshes/F_MED_RaggedRebel_CtrlRig
graphs [
 New Function_ContainedGraph {nodes [
   n1 = RigVMFunctionEntryNode()
   RigVMFunction_ControlFlowBranch = &RigVMFunction_ControlFlowBranch(Condition=@n1.Debug, ExecuteContext=@n1.ExecuteContext)
   RigVMFunctionReturnNode(ExecuteContext=@RigVMFunction_ControlFlowBranch.Completed)
   &RigUnit_SetTransform(
    Item=(Type=Bone,Name="thigh_l")
    Weight=1
    bInitial=false
    bPropagateToChildren=true
    Value=&RigUnit_GetTransform(Item=(Type=Control,Name="thigh_l_ctrl"), bInitial=false).Transform
    ExecutePin=@RigVMFunction_ControlFlowBranch.True
   )
   RigUnit_GetTransform = &RigUnit_GetTransform(Item=(Type=Control,Name="foot_l_ctrl"), bInitial=False)
   > RigUnit_SetTransform = &RigUnit_SetTransform(Item=(Type=Bone,Name="foot_l"), Weight=1, bInitial=False, Value=@RigUnit_GetTransform.Transform)
   > &RigUnit_SetTransform(
    Item=(Type=Bone,Name="foot_r")
    Weight=1
    bInitial=False
    Value=&RigUnit_GetTransform(Item=(Type=Control,Name="foot_r_ctrl"), bInitial=False).Transform
   )
  ]}
 New Function_ContainedGraph {nodes [
   RigVMFunctionEntryNode()
   > &RigUnit_ParentConstraint(
    AdvancedSettings=(InterpolationType=Average,RotationOrderForFilter=XZY)
    Child=(Type=Bone,Name="thigh_l_get_rot")
    Parents=((Item=(Type=Bone,Name="thigh_l"),Weight=1))
    Weight=1
   )
   > get_swing_axis_front = RigVMCollapseNode(
    Axis to Get=(X=0,Y=1,Z=1)
    Backward=false
    Forward=true
    Inward=false
    Outward=true
    Rotation Order=ZXY
    Twist Axis=(X=1,Y=0,Z=0)
    Twist_Inward=false
    Twist_Outward=false
    input joint=thigh_l_get_rot
   )
   > &RigUnit_ModifyTransforms(
    ItemToModify=((Item=(Type=Bone,Name="thigh_l_set_rot_front"),Transform=(Rotation=(X=0,Y=0,Z=0,W=1),Translation=(X=0,Y=0,Z=0),Scale3D=(X=1,Y=1,Z=1))))
    Weight=1
    WeightMinimum=0
    ItemToModify.0.Transform.Rotation=@get_swing_axis_front.Output Rotate.Value
   )
   > &RigUnit_RotationConstraintLocalSpaceOffset(
    AdvancedSettings=(InterpolationType=Average,RotationOrderForFilter=XYZ)
    Child=(Type=Bone,Name="base_strap_front_helper_orient")
    Filter=(bX=True,bY=True,bZ=True)
    Parents=((Item=(Type=Bone,Name="thigh_l_set_rot_front"),Weight=1))
    Weight=1
   )
   RigVMFunction_MathQuaternionToEuler = &RigVMFunction_MathQuaternionToEuler(RotationOrder=ZXY, Value=@get_swing_axis_front.Output Rotate.Value)
   Remap = &RigVMFunction_MathDoubleRemap(SourceMaximum=-110, SourceMinimum=0, TargetMaximum=1, TargetMinimum=0, Value=@RigVMFunction_MathQuaternionToEuler.Result.Z)
   > &RigUnit_ModifyTransforms(
    ItemToModify=((Item=(Type=Bone,Name="base_strap_front_helper_auto"),Transform=(Rotation=(X=-0.4986,Y=0.5861,Z=-0.2575,W=0.5844),Translation=(X=2.6915,Y=-1.8471,Z=-2.4634),Scale3D=(X=1,Y=1,Z=1))))
    WeightMinimum=0
    Weight=@Remap.Result
   )
   > get_swing_axis_back = RigVMCollapseNode(
    Axis to Get=(X=0,Y=1,Z=1)
    Backward=true
    Forward=false
    Inward=false
    Outward=true
    Rotation Order=ZXY
    Twist Axis=(X=1,Y=0,Z=0)
    Twist_Inward=false
    Twist_Outward=false
    input joint=thigh_l_get_rot
   )
   > &RigUnit_ModifyTransforms(
    ItemToModify=((Item=(Type=Bone,Name="thigh_l_set_rot_back"),Transform=(Rotation=(X=0,Y=0,Z=0,W=1),Translation=(X=0,Y=0,Z=0),Scale3D=(X=1,Y=1,Z=1))))
    Weight=1
    WeightMinimum=0
    ItemToModify.0.Transform.Rotation=@get_swing_axis_back.Output Rotate.Value
   )
   > &RigUnit_RotationConstraintLocalSpaceOffset(
    AdvancedSettings=(InterpolationType=Average,RotationOrderForFilter=XYZ)
    Child=(Type=Bone,Name="base_strap_back_0")
    Filter=(bX=True,bY=True,bZ=True)
    Parents=((Item=(Type=Bone,Name="thigh_l_set_rot_back"),Weight=1))
    Weight=1
   )
   > &RigUnit_ModifyTransforms(
    ItemToModify=((Item=(Type=Bone,Name="base_strap_front_helper_offset"),Transform=(Rotation=(X=0,Y=0,Z=0,W=1),Translation=(X=2,Y=0,Z=0),Scale3D=(X=1,Y=1,Z=1))))
    WeightMinimum=0
    Weight=&RigVMFunction_MathDoubleRemap(
     SourceMaximum=70
     SourceMinimum=0
     TargetMaximum=1
     TargetMinimum=0
     Value=&RigVMFunction_MathQuaternionToEuler(RotationOrder=ZXY, Value=@get_swing_axis_back.Output Rotate.Value).Result.Z
    ).Result
   )
   > RigVMFunctionReturnNode()
  ]}
 New Function_ContainedGraph {nodes [
   n1 = RigVMFunctionEntryNode()
   RigUnit_GetTransform = &RigUnit_GetTransform(Item=(Type=Bone,Name="thigh_l_for_cons"), Space=LocalSpace, bInitial=false, Item.Name=@n1.input joint)
   RigVMFunction_MathQuaternionToEuler = &RigVMFunction_MathQuaternionToEuler(
    RotationOrder=@n1.Rotation Order
    Value=&RigVMFunction_MathQuaternionSwingTwist(
     TwistAxis=(X=1,Y=0,Z=0)
     TwistAxis.X=@n1.Twist Axis.X
     TwistAxis.Y=@n1.Twist Axis.Y
     TwistAxis.Z=@n1.Twist Axis.Z
     Input=@RigUnit_GetTransform.Transform.Rotation
    ).Swing
   )
   n2 = &RigVMFunction_MathDoubleMul(B=@RigVMFunction_MathQuaternionToEuler.Result.Z, A=@n1.Axis to Get.Z)
   n3 = /Script/RigVM.RigVMFunction_MathDoubleLess(B=0, A=@n2.Result) {is=struct RigVMUnitNode, template="Less::Execute(in A,in B,out Result)"}
   n4 = /Script/RigVM.RigVMFunction_MathDoubleGreater(B=0, A=@n2.Result) {is=struct RigVMUnitNode, template="Greater::Execute(in A,in B,out Result)"}
   Add = &RigVMFunction_MathDoubleAdd(
    A=RigVMDispatchNode(
     False=0
     Condition=@n1.Forward
     True=RigVMDispatchNode(False=0, Condition=@n3.Result, True=@RigVMFunction_MathQuaternionToEuler.Result.Z) {template="DISPATCH_RigVMDispatch_If(in Condition,in True,in False,out Result)"}.Result
    ) {template="DISPATCH_RigVMDispatch_If(in Condition,in True,in False,out Result)"}.Result
    B=RigVMDispatchNode(
     False=0
     Condition=@n1.Backward
     True=RigVMDispatchNode(False=0, Condition=@n4.Result, True=@RigVMFunction_MathQuaternionToEuler.Result.Z) {template="DISPATCH_RigVMDispatch_If(in Condition,in True,in False,out Result)"}.Result
    ) {template="DISPATCH_RigVMDispatch_If(in Condition,in True,in False,out Result)"}.Result
   )
   n5 = &RigVMFunction_MathDoubleMul(A=@n1.Axis to Get.Y, B=@RigVMFunction_MathQuaternionToEuler.Result.Y)
   Less = &RigVMFunction_MathDoubleLess(B=0, A=@n5.Result) {template="Less::Execute(in A,in B,out Result)"}
   n6 = RigVMDispatchNode(False=0, Condition=@Less.Result, True=@RigVMFunction_MathQuaternionToEuler.Result.Y) {template="DISPATCH_RigVMDispatch_If(in Condition,in True,in False,out Result)"}
   Greater = &RigVMFunction_MathDoubleGreater(B=0, A=@n5.Result) {template="Greater::Execute(in A,in B,out Result)"}
   n7 = RigVMDispatchNode(False=0, Condition=@Greater.Result, True=@RigVMFunction_MathQuaternionToEuler.Result.Y) {template="DISPATCH_RigVMDispatch_If(in Condition,in True,in False,out Result)"}
   RigVMFunction_MathQuaternionFromEuler = &RigVMFunction_MathQuaternionFromEuler(
    Euler=(X=0,Y=0,Z=0)
    Euler.Z=@Add.Result
    Euler.Y=&RigVMFunction_MathDoubleAdd(
     A=RigVMDispatchNode(False=0, Condition=@n1.Outward, True=@n6.Result) {template="DISPATCH_RigVMDispatch_If(in Condition,in True,in False,out Result)"}.Result
     B=RigVMDispatchNode(False=0, Condition=@n1.Inward, True=@n7.Result) {template="DISPATCH_RigVMDispatch_If(in Condition,in True,in False,out Result)"}.Result
    ).Result
    RotationOrder=@n1.Rotation Order
   )
   > RigVMFunctionReturnNode(Output Rotate=(Value=(X=0,Y=0,Z=0,W=1)), Output Rotate.Value=@RigVMFunction_MathQuaternionFromEuler.Result)
  ]}
 New Function_ContainedGraph {nodes [
   n1 = RigVMFunctionEntryNode()
   RigUnit_GetTransform = &RigUnit_GetTransform(Item=(Type=Bone,Name="thigh_l_for_cons"), Space=LocalSpace, bInitial=false, Item.Name=@n1.input joint)
   RigVMFunction_MathQuaternionToEuler = &RigVMFunction_MathQuaternionToEuler(
    RotationOrder=@n1.Rotation Order
    Value=&RigVMFunction_MathQuaternionSwingTwist(
     TwistAxis=(X=1,Y=0,Z=0)
     TwistAxis.X=@n1.Twist Axis.X
     TwistAxis.Y=@n1.Twist Axis.Y
     TwistAxis.Z=@n1.Twist Axis.Z
     Input=@RigUnit_GetTransform.Transform.Rotation
    ).Swing
   )
   n2 = &RigVMFunction_MathDoubleMul(B=@RigVMFunction_MathQuaternionToEuler.Result.Z, A=@n1.Axis to Get.Z)
   n3 = &RigVMFunction_MathDoubleLess(B=0, A=@n2.Result) {template="Less::Execute(in A,in B,out Result)"}
   n4 = &RigVMFunction_MathDoubleGreater(B=0, A=@n2.Result) {template="Greater::Execute(in A,in B,out Result)"}
   Add = &RigVMFunction_MathDoubleAdd(
    A=RigVMDispatchNode(
     False=0
     Condition=@n1.Forward
     True=RigVMDispatchNode(False=0, Condition=@n3.Result, True=@RigVMFunction_MathQuaternionToEuler.Result.Z) {template="DISPATCH_RigVMDispatch_If(in Condition,in True,in False,out Result)"}.Result
    ) {template="DISPATCH_RigVMDispatch_If(in Condition,in True,in False,out Result)"}.Result
    B=RigVMDispatchNode(
     False=0
     Condition=@n1.Backward
     True=RigVMDispatchNode(False=0, Condition=@n4.Result, True=@RigVMFunction_MathQuaternionToEuler.Result.Z) {template="DISPATCH_RigVMDispatch_If(in Condition,in True,in False,out Result)"}.Result
    ) {template="DISPATCH_RigVMDispatch_If(in Condition,in True,in False,out Result)"}.Result
   )
   n5 = &RigVMFunction_MathDoubleMul(A=@n1.Axis to Get.Y, B=@RigVMFunction_MathQuaternionToEuler.Result.Y)
   Less = &RigVMFunction_MathDoubleLess(B=0, A=@n5.Result) {template="Less::Execute(in A,in B,out Result)"}
   n6 = RigVMDispatchNode(False=0, Condition=@Less.Result, True=@RigVMFunction_MathQuaternionToEuler.Result.Y) {template="DISPATCH_RigVMDispatch_If(in Condition,in True,in False,out Result)"}
   Greater = &RigVMFunction_MathDoubleGreater(B=0, A=@n5.Result) {template="Greater::Execute(in A,in B,out Result)"}
   n7 = RigVMDispatchNode(False=0, Condition=@Greater.Result, True=@RigVMFunction_MathQuaternionToEuler.Result.Y) {template="DISPATCH_RigVMDispatch_If(in Condition,in True,in False,out Result)"}
   RigVMFunction_MathQuaternionFromEuler = &RigVMFunction_MathQuaternionFromEuler(
    Euler=(X=0,Y=0,Z=0)
    Euler.Z=@Add.Result
    Euler.Y=&RigVMFunction_MathDoubleAdd(
     A=RigVMDispatchNode(False=0, Condition=@n1.Outward, True=@n6.Result) {template="DISPATCH_RigVMDispatch_If(in Condition,in True,in False,out Result)"}.Result
     B=RigVMDispatchNode(False=0, Condition=@n1.Inward, True=@n7.Result) {template="DISPATCH_RigVMDispatch_If(in Condition,in True,in False,out Result)"}.Result
    ).Result
    RotationOrder=@n1.Rotation Order
   )
   > RigVMFunctionReturnNode(Output Rotate=(Value=(X=0,Y=0,Z=0,W=1)), Output Rotate.Value=@RigVMFunction_MathQuaternionFromEuler.Result)
  ]}
 New Function_ContainedGraph {nodes [
   RigVMFunctionEntryNode()
   > &RigUnit_ParentConstraint(
    AdvancedSettings=(RotationOrderForFilter=XZY)
    Child=(Type=Bone,Name="base_strap_front_helper")
    Parents=((Item=(Type=Bone,Name="spine_02"),Weight=0.0334),(Item=(Type=Bone,Name="thigh_twist_01_l"),Weight=0.0412),(Item=(Type=Bone,Name="deform_groin_l"),Weight=0.2597),(Item=(Type=Bone,Name="pelvis"),Weight=0.6658))
    Weight=1
   )
   > &RigUnit_ParentConstraint(
    AdvancedSettings=(RotationOrderForFilter=XZY)
    Child=(Type=Bone,Name="base_strap_back_0")
    Parents="((Item=(Type=Bone,Name=\"spine_03\"),Weight=0.0014),(Item=(Type=Bone,Name=\"dyn_strap_back_1\"),Weight=0.0028),(Item=(Type=Bone,Name=\"thigh_twist_01_l\"),Weight=0.011),(Item=(Type=Bone,Name=\"dyn_strap_back_0\"),Weight=0.0187),(Item=(Type=Bone,Name=\"spine_01\"),We ...(truncated)"
    Weight=1
   )
   > &RigUnit_ParentConstraint(
    AdvancedSettings=(RotationOrderForFilter=XZY)
    Child=(Type=Bone,Name="base_bag_chest_l")
    Parents="((Item=(Type=Bone,Name=\"deform_upperarm_fr_l\"),Weight=0.0037),(Item=(Type=Bone,Name=\"spine_03\"),Weight=0.0068),(Item=(Type=Bone,Name=\"spine_02\"),Weight=0.0072),(Item=(Type=Bone,Name=\"upperarm_twist_01_l\"),Weight=0.0115),(Item=(Type=Bone,Name=\"clavicle_l\"), ...(truncated)"
    Weight=1
   )
   > &RigUnit_ParentConstraint(
    AdvancedSettings=(RotationOrderForFilter=XZY)
    Child=(Type=Bone,Name="base_bag_chest_r")
    Parents="((Item=(Type=Bone,Name=\"deform_upperarm_fr_r\"),Weight=0.0037),(Item=(Type=Bone,Name=\"spine_03\"),Weight=0.0068),(Item=(Type=Bone,Name=\"spine_02\"),Weight=0.0072),(Item=(Type=Bone,Name=\"upperarm_twist_01_r\"),Weight=0.0115),(Item=(Type=Bone,Name=\"clavicle_r\"), ...(truncated)"
    Weight=1
   )
   > &RigUnit_ParentConstraint(
    AdvancedSettings=(RotationOrderForFilter=XZY)
    Child=(Type=Bone,Name="base_bag_spine_r")
    Parents="((Item=(Type=Bone,Name=\"clavicle_r\"),Weight=0.0121),(Item=(Type=Bone,Name=\"spine_05\"),Weight=0.0151),(Item=(Type=Bone,Name=\"pelvis\"),Weight=0.0487),(Item=(Type=Bone,Name=\"spine_01\"),Weight=0.0536),(Item=(Type=Bone,Name=\"spine_03\"),Weight=0.1283),(Item=(Typ ...(truncated)"
    Weight=1
   )
   > &RigUnit_ParentConstraint(
    AdvancedSettings=(RotationOrderForFilter=XZY)
    Child=(Type=Bone,Name="base_bullets_front_r")
    Parents="((Item=(Type=Bone,Name=\"pelvis\"),Weight=0.0022),(Item=(Type=Bone,Name=\"clavicle_r\"),Weight=0.0098),(Item=(Type=Bone,Name=\"spine_01\"),Weight=0.0104),(Item=(Type=Bone,Name=\"spine_05\"),Weight=0.0201),(Item=(Type=Bone,Name=\"spine_03\"),Weight=0.0694),(Item=(Typ ...(truncated)"
    Weight=1
   )
   > &RigUnit_ParentConstraint(
    AdvancedSettings=(RotationOrderForFilter=XZY)
    Child=(Type=Bone,Name="base_bullets_back_r")
    Parents="((Item=(Type=Bone,Name=\"pelvis\"),Weight=0.0166),(Item=(Type=Bone,Name=\"upperarm_twist_01_r\"),Weight=0.0309),(Item=(Type=Bone,Name=\"spine_01\"),Weight=0.0489),(Item=(Type=Bone,Name=\"spine_05\"),Weight=0.0666),(Item=(Type=Bone,Name=\"clavicle_r\"),Weight=0.0719) ...(truncated)"
    Weight=1
   )
   > &RigUnit_ParentConstraint(
    AdvancedSettings=(RotationOrderForFilter=XZY)
    Child=(Type=Bone,Name="base_bullets_front_l")
    Parents="((Item=(Type=Bone,Name=\"pelvis\"),Weight=0.0022),(Item=(Type=Bone,Name=\"clavicle_l\"),Weight=0.0098),(Item=(Type=Bone,Name=\"spine_01\"),Weight=0.0104),(Item=(Type=Bone,Name=\"spine_05\"),Weight=0.0201),(Item=(Type=Bone,Name=\"spine_03\"),Weight=0.0694),(Item=(Typ ...(truncated)"
    Weight=1
   )
   > &RigUnit_ParentConstraint(
    AdvancedSettings=(RotationOrderForFilter=XZY)
    Child=(Type=Bone,Name="base_scarf_1")
    Parents=((Item=(Type=Bone,Name="neck_01"),Weight=0.1117),(Item=(Type=Bone,Name="spine_05"),Weight=0.1512),(Item=(Type=Bone,Name="head"),Weight=0.3266),(Item=(Type=Bone,Name="neck_02"),Weight=0.4105))
    Weight=1
   )
   > &RigUnit_ParentConstraint(
    AdvancedSettings=(RotationOrderForFilter=XZY)
    Child=(Type=Bone,Name="base_arm_strap_1_l")
    Filter=(TranslationFilter=(bX=True,bY=True,bZ=True),RotationFilter=(bX=true,bY=true,bZ=true),ScaleFilter=(bX=True,bY=True,bZ=True))
    Parents="((Item=(Type=Bone,Name=\"deform_elbow_out_l\"),Weight=0.0001),(Item=(Type=Bone,Name=\"lowerarm_twist_02_l\"),Weight=0.0087),(Item=(Type=Bone,Name=\"lowerarm_l\"),Weight=0.0157),(Item=(Type=Bone,Name=\"deform_elbow_in_l\"),Weight=0.0201),(Item=(Type=Bone,Name=\"uppe ...(truncated)"
    Weight=1
   )
   > &RigUnit_ParentConstraint(
    AdvancedSettings=(RotationOrderForFilter=XZY)
    Child=(Type=Bone,Name="base_arm_strap_2_l")
    Parents="((Item=(Type=Bone,Name=\"upperarm_twist_02_l\"),Weight=0.0007),(Item=(Type=Bone,Name=\"deform_elbow_in_l\"),Weight=0.0039),(Item=(Type=Bone,Name=\"deform_elbow_out_l\"),Weight=0.0269),(Item=(Type=Bone,Name=\"upperarm_l\"),Weight=0.0437),(Item=(Type=Bone,Name=\"lowe ...(truncated)"
    Weight=1
   )
   > &RigUnit_ParentConstraint(
    AdvancedSettings=(RotationOrderForFilter=XZY)
    Child=(Type=Bone,Name="base_arm_strap_3_l")
    Parents=((Item=(Type=Bone,Name="lowerarm_l"),Weight=0.0904),(Item=(Type=Bone,Name="lowerarm_twist_02_l"),Weight=0.3128),(Item=(Type=Bone,Name="lowerarm_twist_01_l"),Weight=0.5968))
    Weight=1
   )
   > &RigUnit_ParentConstraint(
    AdvancedSettings=(RotationOrderForFilter=XZY)
    Child=(Type=Bone,Name="base_arm_strap_1_r")
    Filter=(TranslationFilter=(bX=True,bY=True,bZ=True),RotationFilter=(bX=true,bY=true,bZ=true),ScaleFilter=(bX=True,bY=True,bZ=True))
    Parents="((Item=(Type=Bone,Name=\"deform_elbow_out_r\"),Weight=0.0001),(Item=(Type=Bone,Name=\"lowerarm_twist_02_r\"),Weight=0.0087),(Item=(Type=Bone,Name=\"lowerarm_r\"),Weight=0.0157),(Item=(Type=Bone,Name=\"deform_elbow_in_r\"),Weight=0.0201),(Item=(Type=Bone,Name=\"uppe ...(truncated)"
    Weight=1
   )
   > &RigUnit_ParentConstraint(
    AdvancedSettings=(RotationOrderForFilter=XZY)
    Child=(Type=Bone,Name="base_arm_strap_2_r")
    Parents="((Item=(Type=Bone,Name=\"upperarm_twist_02_r\"),Weight=0.0007),(Item=(Type=Bone,Name=\"deform_elbow_in_r\"),Weight=0.0039),(Item=(Type=Bone,Name=\"deform_elbow_out_r\"),Weight=0.0269),(Item=(Type=Bone,Name=\"upperarm_r\"),Weight=0.0438),(Item=(Type=Bone,Name=\"lowe ...(truncated)"
    Weight=1
   )
   > &RigUnit_ParentConstraint(
    AdvancedSettings=(RotationOrderForFilter=XZY)
    Child=(Type=Bone,Name="base_arm_strap_3_r")
    Parents=((Item=(Type=Bone,Name="lowerarm_r"),Weight=0.0905),(Item=(Type=Bone,Name="lowerarm_twist_02_r"),Weight=0.3128),(Item=(Type=Bone,Name="lowerarm_twist_01_r"),Weight=0.5967))
    Weight=1
   )
   > &RigUnit_ParentConstraint(
    AdvancedSettings=(RotationOrderForFilter=XZY)
    Child=(Type=Bone,Name="for_aim_jnt_1")
    Parents="((Item=(Type=Bone,Name=\"spine_03\"),Weight=0.0005),(Item=(Type=Bone,Name=\"clavicle_l\"),Weight=0.0015),(Item=(Type=Bone,Name=\"deform_upperarm_up_r\"),Weight=0.0185),(Item=(Type=Bone,Name=\"spine_04\"),Weight=0.0746),(Item=(Type=Bone,Name=\"deform_upperarm_bk_r\") ...(truncated)"
    Weight=1
   )
   > &RigUnit_ParentConstraint(
    AdvancedSettings=(RotationOrderForFilter=XZY)
    Child=(Type=Bone,Name="for_aim_jnt_3")
    Parents="((Item=(Type=Bone,Name=\"pelvis\"),Weight=0.0139),(Item=(Type=Bone,Name=\"upperarm_twist_01_r\"),Weight=0.0186),(Item=(Type=Bone,Name=\"spine_01\"),Weight=0.0583),(Item=(Type=Bone,Name=\"clavicle_r\"),Weight=0.0959),(Item=(Type=Bone,Name=\"spine_05\"),Weight=0.0987) ...(truncated)"
    Weight=1
   )
   > RigVMFunctionReturnNode()
  ]}
 New Function_ContainedGraph {nodes [
   RigVMFunctionEntryNode()
   > DeltaRotation = RigVMFunctionReferenceNode(
    Rotation Order=ZXY
    child=foot_l
    debugRotation=false
    max=-50
    parent=foot_l_null
    parentType=Null
    premult=1
    primAxis=2
    twistAxis=(X=1,Y=0,Z=0)
   )
   > &RigUnit_ModifyTransforms(
    ItemToModify=((Item=(Type=Bone,Name="base_tongue_shoe_l"),Transform=(Rotation=(X=0,Y=-0.0175,Z=0,W=-0.9998),Translation=(X=5.6691,Y=0,Z=-1.9952),Scale3D=(X=1,Y=1,Z=1))))
    WeightMinimum=0
    Weight=@DeltaRotation.outputWeight
   )
   > DeltaRotation_1 = RigVMFunctionReferenceNode(
    Rotation Order=ZXY
    child=foot_r
    debugRotation=false
    max=-50
    parent=foot_r_null
    parentType=Null
    premult=1
    primAxis=2
    twistAxis=(X=1,Y=0,Z=0)
   )
   > &RigUnit_ModifyTransforms(
    ItemToModify=((Item=(Type=Bone,Name="base_tongue_shoe_r"),Transform=(Rotation=(X=0.0361,Y=0.573,Z=0.0515,W=0.8171),Translation=(X=-0.011,Y=0.0262,Z=-3.7698),Scale3D=(X=1,Y=1,Z=1))))
    WeightMinimum=0
    Weight=@DeltaRotation_1.outputWeight
   )
   > RigVMFunctionReturnNode()
  ]}
 New Function_ContainedGraph {nodes [
   n1 = RigVMFunctionEntryNode()
   RigVMFunction_MathBoolToInteger = &RigVMFunction_MathBoolToInteger(Value=@n1.onlyTwist)
   RigVMFunction_MathQuaternionSwingTwist = &RigVMFunction_MathQuaternionSwingTwist(
    twistAxis=@n1.twistAxis
    Input=&RigUnit_GetRelativeTransformForItem(
     bChildInitial=False
     bParentInitial=False
     child=(Type=Bone,Name="None")
     parent=(Type=Bone,Name="None")
     child.Name=@n1.child
     parent.Name=@n1.parent
     parent.Type=@n1.parentType
    ).RelativeTransform.Rotation
   )
   RigVMFunction_MathQuaternionToEuler = &RigVMFunction_MathQuaternionToEuler(
    Value=RigVMDispatchNode(
     Values=((),())
     Index=@RigVMFunction_MathBoolToInteger.Result
     Values.0=@RigVMFunction_MathQuaternionSwingTwist.Swing
     Values.1=@RigVMFunction_MathQuaternionSwingTwist.Twist
    ) {template="DISPATCH_RigVMDispatch_SelectInt32(in Index,in Values,out Result)"}.Result
    RotationOrder=@n1.Rotation Order
   )
   > DISPATCH_RigVMDispatch_Print = RigVMDispatchNode(ScreenColor=(R=1,G=1,B=1,A=1), ScreenDuration=1.5, Enabled=@n1.debugRotation, Value=@RigVMFunction_MathQuaternionToEuler.Result) {template="DISPATCH_RigVMDispatch_Print(in Prefix,in Value,in Enabled,in ScreenDuration,in ScreenColor)"}
   Clamp = &RigVMFunction_MathIntClamp(Value=@n1.primAxis)
   DISPATCH_RigVMDispatch_SelectInt32 = RigVMDispatchNode(
    Values=(0,0,0)
    Index=@Clamp.Result
    Values.0=@RigVMFunction_MathQuaternionToEuler.Result.X
    Values.1=@RigVMFunction_MathQuaternionToEuler.Result.Y
    Values.2=@RigVMFunction_MathQuaternionToEuler.Result.Z
   ) {template="DISPATCH_RigVMDispatch_SelectInt32(in Index,in Values,out Result)"}
   Remap = &RigVMFunction_MathDoubleRemap(
    TargetMaximum=1
    TargetMinimum=0
    bClamp=False
    Value=@DISPATCH_RigVMDispatch_SelectInt32.Result
    SourceMaximum=@n1.max
    SourceMinimum=@n1.min
   )
   Multiply = &RigVMFunction_MathDoubleMul(A=&RigVMFunction_MathDoubleClamp(Value=@Remap.Result).Result, B=@n1.premult)
   > RigVMFunctionReturnNode(outputWeight=@Multiply.Result)
  ]}
 New Function_ContainedGraph {nodes [
   RigVMFunctionEntryNode()
   > &RigUnit_AimConstraintLocalSpaceOffset(
    AdvancedSettings=(DebugSettings=(bEnabled=false,Scale=10,WorldOffset=(Rotation=(X=0,Y=0,Z=0,W=1),Translation=(X=0,Y=0,Z=0),Scale3D=(X=1,Y=1,Z=1))),RotationOrderForFilter=XYZ)
    Child=(Type=Bone,Name="base_scarf_1")
    Filter=(bX=True,bY=True,bZ=True)
    Parents=((Item=(Type=Bone,Name="for_aim_jnt_1"),Weight=1))
    Weight=1
    WorldUp=(Target=(X=0,Y=0,Z=1),Kind=Location,Space=(Type=Bone,Name="for_aim_jnt_1"))
   )
   > DeltaRotation = RigVMFunctionReferenceNode(Child=Head, Parent=neck_01, Source Max=-60, Source Min=-10)
   > &RigUnit_ModifyTransforms(
    ItemToModify=((Item=(Type=Bone,Name="dyn_neck_support_l"),Transform=(Rotation=(X=-0.6209,Y=0.1983,Z=0.1664,W=0.7399),Translation=(X=0,Y=0,Z=0),Scale3D=(X=1,Y=1,Z=1))))
    WeightMinimum=0
    Weight=@DeltaRotation.Delta_X_Negative
   )
   > DeltaRotation_1 = RigVMFunctionReferenceNode(Child=Head, Parent=neck_01, Source Max=-60, Source Min=-10)
   > &RigUnit_ModifyTransforms(
    ItemToModify=((Item=(Type=Bone,Name="dyn_neck_support_r"),Transform=(Rotation=(X=0.6209,Y=-0.1983,Z=0.1664,W=0.7399),Translation=(X=0,Y=0,Z=0),Scale3D=(X=1,Y=1,Z=1))))
    WeightMinimum=0
    Weight=@DeltaRotation_1.Delta_X_Positive
   )
   > RigVMFunctionReturnNode()
  ]}
 New Function_ContainedGraph {nodes [
   RigVMFunctionEntryNode()
   > &RigUnit_ParentConstraint(
    AdvancedSettings=(RotationOrderForFilter=XZY)
    Child=(Type=Bone,Name="base_bullets_back_l")
    Filter=(TranslationFilter=(bX=True,bY=True,bZ=True),RotationFilter=(bX=true,bY=true,bZ=true),ScaleFilter=(bX=True,bY=True,bZ=True))
    Parents="((Item=(Type=Bone,Name=\"pelvis\"),Weight=0.0166),(Item=(Type=Bone,Name=\"upperarm_twist_01_l\"),Weight=0.0309),(Item=(Type=Bone,Name=\"spine_01\"),Weight=0.0489),(Item=(Type=Bone,Name=\"spine_05\"),Weight=0.0666),(Item=(Type=Bone,Name=\"clavicle_l\"),Weight=0.0719) ...(truncated)"
    Weight=1
   )
   > &RigUnit_ParentConstraint(
    AdvancedSettings=(RotationOrderForFilter=XZY)
    Child=(Type=Bone,Name="for_aim_bullets_back_l")
    Filter=(TranslationFilter=(bX=True,bY=True,bZ=True),RotationFilter=(bX=true,bY=true,bZ=true),ScaleFilter=(bX=True,bY=True,bZ=True))
    Parents="((Item=(Type=Bone,Name=\"clavicle_r\"),Weight=0.0082),(Item=(Type=Bone,Name=\"spine_01\"),Weight=0.0313),(Item=(Type=Bone,Name=\"upperarm_twist_01_l\"),Weight=0.0567),(Item=(Type=Bone,Name=\"spine_03\"),Weight=0.0928),(Item=(Type=Bone,Name=\"spine_02\"),Weight=0.121 ...(truncated)"
    Weight=1
   )
   > &RigUnit_AimConstraintLocalSpaceOffset(
    AimAxis=(X=0,Y=1,Z=0)
    Child=(Type=Bone,Name="dyn_bullets_back_l")
    Filter=(bX=true,bY=true,bZ=true)
    Parents=((Item=(Type=Bone,Name="for_aim_bullets_back_l"),Weight=1))
    Weight=1
    WorldUp=(Target=(X=0,Y=0,Z=1),Kind=Direction,Space=(Type=Bone,Name="for_aim_bullets_back_l"))
   )
   > RigVMFunctionReturnNode()
  ]}
 New Function_ContainedGraph {nodes [
   RigVMFunctionEntryNode()
   > n1 = &RigUnit_SphericalPoseReader(
    ActiveRegionScaleFactors=(PositiveWidth=0.1,NegativeWidth=0.1,PositiveHeight=0.1,NegativeHeight=0.1)
    ActiveRegionSize=0.1
    Debug=(bDrawDebug=false,bDraw2D=false,DebugScale=25,DebugSegments=20,DebugThickness=0.25)
    DriverAxis=(X=1,Y=0,Z=0)
    DriverItem=(Type=Bone,Name="calf_r")
    FalloffRegionScaleFactors=(PositiveWidth=0,NegativeWidth=0.7,PositiveHeight=0.7,NegativeHeight=0.7)
    FalloffSize=0.45
    OptionalParentItem=(Type=Bone,Name="thigh_r")
    RotationOffset=(X=-90,Y=0,Z=-90)
   )
   > &RigUnit_ModifyTransforms(
    ItemToModify=((Item=(Type=Bone,Name="calf_counter_r"),Transform=(Rotation=(X=0,Y=0,Z=-0.2588,W=0.9659),Translation=(X=0,Y=0,Z=0),Scale3D=(X=1,Y=1,Z=1))))
    WeightMinimum=0
    Weight=@n1.OutputParam
   )
   > &RigUnit_ModifyTransforms(
    ItemToModify=((Item=(Type=Bone,Name="thigh_counter_r"),Transform=(Rotation=(X=0,Y=0,Z=0.2588,W=0.9659),Translation=(X=0,Y=-10,Z=0),Scale3D=(X=1,Y=1,Z=1))))
    WeightMinimum=0
    Weight=@n1.OutputParam
   )
   > RigVMFunctionReturnNode()
  ]}
 New Function_ContainedGraph {nodes [
   n1 = RigVMFunctionEntryNode()
   RigVMFunction_MathQuaternionToEuler = &RigVMFunction_MathQuaternionToEuler(
    RotationOrder=ZYX
    Value=&RigUnit_GetRelativeTransformForItem(
     Child=(Type=Bone,Name="head")
     Parent=(Type=Bone,Name="neck_01")
     bChildInitial=False
     bParentInitial=False
     Child.Name=@n1.Child
     Parent.Name=@n1.Parent
    ).RelativeTransform.Rotation
   )
   Remap = &RigVMFunction_MathDoubleRemap(
    TargetMaximum=1
    TargetMinimum=0
    SourceMaximum=@n1.Source Max
    SourceMinimum=@n1.Source Min
    Value=@RigVMFunction_MathQuaternionToEuler.Result.Z
   )
   > RigVMFunctionReturnNode(
    Delta_z=@Remap.Result
    Delta_X_Negative=&RigVMFunction_MathDoubleRemap(SourceMaximum=-60, SourceMinimum=0, TargetMaximum=1, TargetMinimum=0, Value=@RigVMFunction_MathQuaternionToEuler.Result.X).Result
    Delta_X_Positive=&RigVMFunction_MathDoubleRemap(SourceMaximum=70, SourceMinimum=0, TargetMaximum=1, TargetMinimum=0, Value=@RigVMFunction_MathQuaternionToEuler.Result.X).Result
   )
  ]}
 New Function_ContainedGraph {nodes [
   RigVMFunctionEntryNode()
   > &RigUnit_ParentConstraint(
    AdvancedSettings=(InterpolationType=Average,RotationOrderForFilter=XZY)
    Child=(Type=Bone,Name="base_bag_spine_l")
    Parents=((Item=(Type=Bone,Name="base_bullets_back_l"),Weight=1),(Item=(Type=Bone,Name="base_bullets_front_l"),Weight=1))
    Weight=1
   )
   > &RigUnit_ParentConstraint(
    AdvancedSettings=(InterpolationType=Average,RotationOrderForFilter=XZY)
    Child=(Type=Bone,Name="base_bag_spine_r")
    Parents=((Item=(Type=Bone,Name="base_bullets_back_r"),Weight=1),(Item=(Type=Bone,Name="base_bullets_front_r"),Weight=1))
    Weight=1
   )
   > n1 = &RigUnit_SphericalPoseReader(
    ActiveRegionScaleFactors=(PositiveWidth=1,NegativeWidth=1,PositiveHeight=2,NegativeHeight=1)
    ActiveRegionSize=0.05
    Debug=(bDrawDebug=false,bDraw2D=false,bDrawLocalAxes=true,DebugScale=25,DebugSegments=20,DebugThickness=0.25)
    DriverAxis=(X=-1,Y=0,Z=0)
    DriverItem=(Type=Bone,Name="upperarm_r")
    FalloffRegionScaleFactors=(PositiveWidth=0.5,NegativeWidth=0.5,PositiveHeight=1,NegativeHeight=1)
    FalloffSize=0.3
    FlipWidthScaling=true
    OptionalParentItem=(Type=Bone,Name="spine_04")
    RotationOffset=(X=-50,Y=35,Z=20)
   )
   AlphaInterp = &RigVMFunction_AlphaInterp(InterpSpeedDecreasing=10, InterpSpeedIncreasing=10, Scale=1, bClampResult=true, bMapRange=true, Value=@n1.OutputParam)
   > &RigUnit_ModifyTransforms(
    ItemToModify="((Item=(Type=Bone,Name=\"dyn_radio\"),Transform=(Rotation=(X=0.1854,Y=0.0915,Z=0.4402,W=-0.8738),Translation=(X=-5.8308,Y=-4.4786,Z=1.972),Scale3D=(X=1,Y=1,Z=1))),(Item=(Type=Bone,Name=\"dyn_bag_chest_r\"),Transform=(Rotation=(X=-0.1792,Y=-0.0318,Z=-0.2229,W=0 ...(truncated)"
    WeightMinimum=0
    Weight=@AlphaInterp.Result
   )
   > RigVMFunctionReturnNode()
  ]}
 RigVMFunctionLibrary {nodes [
   DeltaRotation = RigVMCollapseNode(ExecuteContext=(), Rotation Order=ZYX, premult=1, twistAxis=(X=1,Y=0,Z=0))
   DeltaRotation_1 = RigVMCollapseNode(ExecuteContext=())
   bags radio = RigVMCollapseNode(ExecuteContext=())
   bullets = RigVMCollapseNode(ExecuteContext=())
   controls = RigVMCollapseNode(Debug=False, ExecuteContext=())
   knee counter = RigVMCollapseNode(ExecuteContext=())
   scarf = RigVMCollapseNode(ExecuteContext=())
   strap = RigVMCollapseNode(ExecuteContext=())
   surface constraints = RigVMCollapseNode(ExecuteContext=())
   tongues of footwear = RigVMCollapseNode(ExecuteContext=())
  ]}
 RigVMModel {nodes [
   RigUnit_BeginExecution = &RigUnit_BeginExecution()
   controls = RigVMFunctionReferenceNode(Debug=false, ExecuteContext=@RigUnit_BeginExecution.ExecutePin)
   > FN_DeformRig_FNC = RigVMFunctionReferenceNode(
    DeformRig_ Curve="(deformRig_l_upperarm_up_alpha_72_#1=False,deformRig_r_upperarm_up_alpha_73_#2=False,deformRig_l_upperarm_fr_alpha_74_#3=False,deformRig_r_upperarm_fr_alpha_75_CB51B8 ...(truncated)"
    DeformRig_Bool="(upperarm_1_#4=True,elbow_3_#5=True,wrist_7_#6=True,groin_8_#7=True,glute_17_#8=True,knee_9_40C408324CA261 ...(truncated)"
    DeformRig_Debug=(upperarm_debug_18_#4=false,wrist_debug_20_#6=False,lat_debug_26_#9=false,groinGlute_debug_27_#7=false)
    DeformRig_Value="(upperarm_up_18_#10=(Rotation=(X=0.0,Y=-0.0,Z=0.0,W=1.0),Translation=(X=0.0,Y=0.0,Z=8.0),Scale3D=(X=0.70,Y=1.0,Z=1.0)),upperarm_fwd_fr_19_#11=(Rotation=(X=0.0,Y=-0.0,Z=0.0,W=1.0),Translation=(X=0.0, ...(truncated)"
   )
   > surface constraints = RigVMFunctionReferenceNode()
   > strap = RigVMFunctionReferenceNode()
   > tongues of footwear = RigVMFunctionReferenceNode()
   > scarf = RigVMFunctionReferenceNode()
   > bullets = RigVMFunctionReferenceNode()
   > knee counter = RigVMFunctionReferenceNode()
   > bags radio = RigVMFunctionReferenceNode()
  ]}
]
hierarchy [
 root {loc=<0 0 0>, rot=<0 0 0 1>}
 pelvis {parent=root, loc=<0 -0.5301 98.0571>, rot=<0.0127 0.707 -0.0127 -0.707>}
 spine_01 {parent=pelvis, loc=<6.1208 -0.9328 0>, rot=<0 0 0.034 -0.9994>}
 spine_02 {parent=spine_01, loc=<6.5997 -0.2095 0>, rot=<0 0 0.0238 -0.9997>}
 spine_03 {parent=spine_02, loc=<6.4914 -0.1237 0>, rot=<0 0 0.0149 -0.9999>}
 spine_04 {parent=spine_03, loc=<6.9598 0.0794 0>, rot=<0 0 -0.0132 -0.9999>}
 spine_05 {parent=spine_04, loc=<6.9987 0 0>, rot=<0 0 -0.0015 -1>}
 neck_01 {parent=spine_05, loc=<11.3138 0.1529 0>, rot=<0 0 -0.2003 -0.9797>}
 neck_02 {parent=neck_01, loc=<4.5627 0 0>, rot=<0 0 0.0504 -0.9987>}
 head {parent=neck_02, loc=<5.3914 0 0>, rot=<0 0 0.115 -0.9934>}
 base_scarf_1 {parent=neck_02, loc=<0.4026 -4.2421 -0.1882>, rot=<0.2825 -0.4911 0.8179 0.1003>}
 dyn_scarf_1 {parent=base_scarf_1, loc=<0 0 0>, rot=<0 0 0 1>}
 dyn_scarf_2 {parent=dyn_scarf_1, loc=<6.1891 0 0>, rot=<0.0433 0.0588 -0.1129 -0.9909>}
 dyn_scarf_3 {parent=dyn_scarf_2, loc=<7.958 0 0>, rot=<0.1001 -0.1931 -0.1075 -0.9701>}
 dyn_scarf_4 {parent=dyn_scarf_3, loc=<8.0853 0 0>, rot=<0.0106 -0.0374 -0.0256 -0.9989>}
 dyn_scarf_5 {parent=dyn_scarf_4, loc=<8.1102 0 0>, rot=<0 0 0 1>}
 dyn_scarf_6 {parent=dyn_scarf_5, loc=<7.8672 0.1773 -0.032>, rot=<0 0 0 1>}
 base_neck_support {parent=neck_02, loc=<-0.8999 -3.9838 0>, rot=<0 0 0 1>}
 dyn_neck_support {parent=base_neck_support, loc=<0 0 0>, rot=<0 0 0 1>}
 base_neck_support_l {parent=neck_02, loc=<-0.8999 -3.9838 0>, rot=<0 0 0 1>}
 dyn_neck_support_l {parent=base_neck_support_l, loc=<0 0 0>, rot=<0 0 0 1>}
 base_neck_support_r {parent=neck_02, loc=<-0.8999 -3.9838 0>, rot=<0 0 0 1>}
 dyn_neck_support_r {parent=base_neck_support_r, loc=<0 0 0>, rot=<0 0 0 1>}
 clavicle_l {parent=spine_05, loc=<5.4831 -0.1053 -3.2946>, rot=<0.0465 -0.7289 -0.009 -0.6829>}
 upperarm_l {parent=clavicle_l, loc=<9.8658 -1.2136 -0.4105>, rot=<0.0164 -0.3245 -0.0269 -0.9454>}
 lowerarm_l {parent=upperarm_l, loc=<26.465 -0.8345 0.259>, rot=<0.0036 0.0161 -0.2298 -0.9731>}
 hand_l {parent=lowerarm_l, loc=<22.7355 -0.3203 -0.6105>, rot=<-0.6784 -0.0223 0.0735 -0.7307>}
 thumb_01_l {parent=hand_l, loc=<1.3118 -0.3191 -2.459>, rot=<0.6546 -0.4042 0.0664 -0.6354>}
 thumb_02_l {parent=thumb_01_l, loc=<5.9336 -0.0448 0.0401>, rot=<-0.0037 0.0213 0.298 -0.9543>}
 thumb_03_l {parent=thumb_02_l, loc=<2.2223 0.3682 -0.0907>, rot=<0.0118 -0.013 0.0066 -0.9998>}
 index_metacarpal_l {parent=hand_l, loc=<1.6535 0.0183 -2.96>, rot=<0.0158 -0.0124 -0.0173 -0.9996>}
 index_01_l {parent=index_metacarpal_l, loc=<8.3008 0.0357 -0.0524>, rot=<-0.0044 0.0485 0.0924 -0.9945>}
 index_02_l {parent=index_01_l, loc=<3.3587 -0.3004 0.0452>, rot=<-0.002 0.0027 0.3225 -0.9466>}
 index_03_l {parent=index_02_l, loc=<2.5578 -0.123 -0.0252>, rot=<-0.002 0.0067 0.2558 -0.9667>}
 middle_metacarpal_l {parent=hand_l, loc=<2.1856 -0.1725 -0.5984>, rot=<0.006 0.0168 -0.0329 -0.9993>}
 middle_01_l {parent=middle_metacarpal_l, loc=<7.3275 -0.2447 0.0129>, rot=<-0.0208 0.0042 0.0858 -0.9961>}
 middle_02_l {parent=middle_01_l, loc=<4.0123 0.1401 0.1175>, rot=<-0.0218 0.0258 0.3512 -0.9357>}
 middle_03_l {parent=middle_02_l, loc=<2.6077 0.0534 0.1373>, rot=<0.0082 0.0188 0.1708 -0.9851>}
 ring_metacarpal_l {parent=hand_l, loc=<2.2074 -0.3625 1.3638>, rot=<-0.0284 0.0722 -0.0458 -0.9959>}
 ring_01_l {parent=ring_metacarpal_l, loc=<7.0295 -0.6197 -0.1446>, rot=<-0.0489 -0.0252 0.1892 -0.9804>}
 ring_02_l {parent=ring_01_l, loc=<3.6174 0.4647 -0.1027>, rot=<0 0 0.2806 -0.9598>}
 ring_03_l {parent=ring_02_l, loc=<2.4152 0.114 0.0199>, rot=<0 0 0.1189 -0.9929>}
 pinky_metacarpal_l {parent=hand_l, loc=<1.7616 -0.9753 3.3814>, rot=<-0.0397 0.0814 -0.069 -0.9935>}
 pinky_01_l {parent=pinky_metacarpal_l, loc=<7.1709 -0.4831 -0.1094>, rot=<-0.0654 -0.0035 0.2242 -0.9723>}
 pinky_02_l {parent=pinky_01_l, loc=<2.8274 0.0893 -0.1043>, rot=<0.0037 0.0011 0.2854 -0.9584>}
 pinky_03_l {parent=pinky_02_l, loc=<2.2046 0.0307 -0.0078>, rot=<-0.0002 0.0098 0.1445 -0.9895>}
 deform_wrist_dn_base_l {parent=hand_l, loc=<0 -1.5326 0>, rot=<0 0 0 -1>}
 deform_wrist_dn_l {parent=deform_wrist_dn_base_l, loc=<0 0 0>, rot=<0 0 0 1>}
 deform_wrist_up_base_l {parent=hand_l, loc=<0 2.3009 0>, rot=<0 0 0 -1>}
 deform_wrist_up_l {parent=deform_wrist_up_base_l, loc=<0 0 0>, rot=<0 0 0 1>}
 lowerarm_twist_02_l {parent=lowerarm_l, loc=<10.6707 0 0>, rot=<0 0 0 1>}
 lowerarm_twist_01_l {parent=lowerarm_l, loc=<18.316 0 0>, rot=<0 0 0 1>}
 deform_elbow_in_base_l {parent=lowerarm_l, loc=<0.1834 3.6062 -0.2572>, rot=<0 0 0 -1>}
 deform_elbow_in_l {parent=deform_elbow_in_base_l, loc=<0 0 0>, rot=<0 0 0 1>}
 deform_elbow_out_base_l {parent=lowerarm_l, loc=<0 -4 0>, rot=<0 0 0 1>}
 deform_elbow_out_l {parent=deform_elbow_out_base_l, loc=<0 0 0>, rot=<0 0 0 1>}
 base_arm_strap_2_l {parent=lowerarm_l, loc=<5.2977 -0.4106 4.4724>, rot=<0.6222 -0.4172 0.5621 -0.3505>}
 dyn_arm_strap_2_l {parent=base_arm_strap_2_l, loc=<0 0 0>, rot=<0 0 0 1>}
 base_arm_strap_3_l {parent=lowerarm_l, loc=<12.5752 -0.5959 3.3541>, rot=<0.6034 -0.444 0.5463 -0.3747>}
 dyn_arm_strap_3_l {parent=base_arm_strap_3_l, loc=<0 0 0>, rot=<0 0 0 1>}
 upperarm_twist_01_l {parent=upperarm_l, loc=<7.2135 -0.1134 0>, rot=<0 0 0 1>}
 deform_upperarm_bk_base_l {parent=upperarm_twist_01_l, loc=<-6.4053 -4.3746 0.314>, rot=<0 0 0 -1>}
 deform_upperarm_bk_l {parent=deform_upperarm_bk_base_l, loc=<0 0 0>, rot=<0 0 0 1>}
 deform_upperarm_fr_base_l {parent=upperarm_twist_01_l, loc=<-5.6312 5.092 0.3354>, rot=<0 0 0 -1>}
 deform_upperarm_fr_l {parent=deform_upperarm_fr_base_l, loc=<0 0 0>, rot=<0 0 0 1>}
 deform_upperarm_up_base_l {parent=upperarm_twist_01_l, loc=<-8.891 0.4825 3.9517>, rot=<0 0 0 -1>}
 deform_upperarm_up_l {parent=deform_upperarm_up_base_l, loc=<0 0 0>, rot=<0 0 0 1>}
 upperarm_twist_02_l {parent=upperarm_l, loc=<19.7998 -0.1134 0>, rot=<0 0 0 1>}
 base_arm_strap_1_l {parent=upperarm_l, loc=<19.3996 -0.0726 3.6663>, rot=<0.5573 -0.3626 0.6316 -0.3988>}
 dyn_arm_strap_1_l {parent=base_arm_strap_1_l, loc=<0 0 0>, rot=<0 0 0 1>}
 clavicle_r {parent=spine_05, loc=<5.4831 -0.1053 3.2946>, rot=<0.7289 0.0465 -0.6829 0.009>}
 upperarm_r {parent=clavicle_r, loc=<-9.8658 1.2136 0.4105>, rot=<0.0164 -0.3245 -0.0269 -0.9454>}
 lowerarm_r {parent=upperarm_r, loc=<-26.465 0.8345 -0.259>, rot=<0.0036 0.0161 -0.2298 -0.9731>}
 hand_r {parent=lowerarm_r, loc=<-22.7355 0.3203 0.6106>, rot=<-0.6784 -0.0223 0.0735 -0.7307>}
 thumb_01_r {parent=hand_r, loc=<-1.3118 0.3191 2.459>, rot=<0.6546 -0.4042 0.0664 -0.6354>}
 thumb_02_r {parent=thumb_01_r, loc=<-5.9336 0.0448 -0.0401>, rot=<-0.0037 0.0213 0.298 -0.9543>}
 thumb_03_r {parent=thumb_02_r, loc=<-2.2223 -0.3682 0.0907>, rot=<0.0118 -0.013 0.0066 -0.9998>}
 index_metacarpal_r {parent=hand_r, loc=<-1.6535 -0.0183 2.96>, rot=<0.0158 -0.0124 -0.0173 -0.9996>}
 index_01_r {parent=index_metacarpal_r, loc=<-8.3008 -0.0357 0.0524>, rot=<-0.0044 0.0485 0.0924 -0.9945>}
 index_02_r {parent=index_01_r, loc=<-3.3587 0.3004 -0.0452>, rot=<-0.002 0.0027 0.3225 -0.9466>}
 index_03_r {parent=index_02_r, loc=<-2.5578 0.123 0.0252>, rot=<-0.002 0.0067 0.2558 -0.9667>}
 middle_metacarpal_r {parent=hand_r, loc=<-2.1856 0.1725 0.5984>, rot=<0.006 0.0168 -0.0329 -0.9993>}
 middle_01_r {parent=middle_metacarpal_r, loc=<-7.3275 0.2447 -0.0129>, rot=<-0.0208 0.0042 0.0858 -0.9961>}
 middle_02_r {parent=middle_01_r, loc=<-4.0123 -0.1401 -0.1175>, rot=<-0.0218 0.0258 0.3512 -0.9357>}
 middle_03_r {parent=middle_02_r, loc=<-2.6077 -0.0534 -0.1373>, rot=<0.0082 0.0188 0.1708 -0.9851>}
 ring_metacarpal_r {parent=hand_r, loc=<-2.2074 0.3625 -1.3638>, rot=<-0.0284 0.0722 -0.0458 -0.9959>}
 ring_01_r {parent=ring_metacarpal_r, loc=<-7.0295 0.6197 0.1446>, rot=<-0.0489 -0.0252 0.1892 -0.9804>}
 ring_02_r {parent=ring_01_r, loc=<-3.6174 -0.4647 0.1027>, rot=<0 0 0.2806 -0.9598>}
 ring_03_r {parent=ring_02_r, loc=<-2.4152 -0.114 -0.0199>, rot=<0 0 0.1189 -0.9929>}
 pinky_metacarpal_r {parent=hand_r, loc=<-1.7616 0.9753 -3.3814>, rot=<-0.0397 0.0814 -0.069 -0.9935>}
 pinky_01_r {parent=pinky_metacarpal_r, loc=<-7.1709 0.4831 0.1094>, rot=<-0.0654 -0.0035 0.2242 -0.9723>}
 pinky_02_r {parent=pinky_01_r, loc=<-2.8274 -0.0893 0.1043>, rot=<0.0037 0.0011 0.2854 -0.9584>}
 pinky_03_r {parent=pinky_02_r, loc=<-2.2046 -0.0307 0.0078>, rot=<-0.0002 0.0098 0.1445 -0.9895>}
 deform_wrist_dn_base_r {parent=hand_r, loc=<-0.0001 1.5328 0>, rot=<0 0 0 -1>}
 deform_wrist_dn_r {parent=deform_wrist_dn_base_r, loc=<0 0 0>, rot=<0 0 0 1>}
 deform_wrist_up_base_r {parent=hand_r, loc=<-0.0002 -2.3007 -0.0001>, rot=<0 0 0 -1>}
 deform_wrist_up_r {parent=deform_wrist_up_base_r, loc=<0 0 0>, rot=<0 0 0 1>}
 lowerarm_twist_02_r {parent=lowerarm_r, loc=<-10.6707 0 0>, rot=<0 0 0 -1>}
 lowerarm_twist_01_r {parent=lowerarm_r, loc=<-18.316 0 0>, rot=<0 0 0 -1>}
 deform_elbow_in_base_r {parent=lowerarm_r, loc=<-0.1834 -3.6062 0.2572>, rot=<0 0 0 -1>}
 deform_elbow_in_r {parent=deform_elbow_in_base_r, loc=<0 0 0>, rot=<0 0 0 1>}
 deform_elbow_out_base_r {parent=lowerarm_r, loc=<0 4 -0.0001>, rot=<0 0 0 -1>}
 deform_elbow_out_r {parent=deform_elbow_out_base_r, loc=<0 0 0>, rot=<0 0 0 1>}
 base_arm_strap_3_r {parent=lowerarm_r, loc=<-12.5755 0.596 -3.3538>, rot=<-0.444 -0.6034 -0.3747 -0.5463>}
 dyn_arm_strap_3_r {parent=base_arm_strap_3_r, loc=<0 0 0>, rot=<0 0 0 1>}
 base_arm_strap_2_r {parent=lowerarm_r, loc=<-5.2977 0.4106 -4.4724>, rot=<-0.4172 -0.6222 -0.3505 -0.5621>}
 dyn_arm_strap_2_r {parent=base_arm_strap_2_r, loc=<0 0 0>, rot=<0 0 0 1>}
 upperarm_twist_01_r {parent=upperarm_r, loc=<-7.2135 0.1134 0>, rot=<0 0 0 -1>}
 deform_upperarm_bk_base_r {parent=upperarm_twist_01_r, loc=<6.4053 4.3746 -0.314>, rot=<0 0 0 -1>}
 deform_upperarm_bk_r {parent=deform_upperarm_bk_base_r, loc=<0 0 0>, rot=<0 0 0 1>}
 deform_upperarm_fr_base_r {parent=upperarm_twist_01_r, loc=<5.6312 -5.092 -0.3354>, rot=<0 0 0 -1>}
 deform_upperarm_fr_r {parent=deform_upperarm_fr_base_r, loc=<0 0 0>, rot=<0 0 0 1>}
 deform_upperarm_up_base_r {parent=upperarm_twist_01_r, loc=<8.891 -0.4825 -3.9517>, rot=<0 0 0 -1>}
 deform_upperarm_up_r {parent=deform_upperarm_up_base_r, loc=<0 0 0>, rot=<0 0 0 1>}
 upperarm_twist_02_r {parent=upperarm_r, loc=<-19.7998 0.1134 0>, rot=<0 0 0 -1>}
 base_arm_strap_1_r {parent=upperarm_r, loc=<-19.3993 0.0726 -3.6665>, rot=<-0.3626 -0.5573 -0.3988 -0.6316>}
 dyn_arm_strap_1_r {parent=base_arm_strap_1_r, loc=<0 0 0>, rot=<0 0 0 1>}
 attach_backpack {parent=spine_05, loc=<-0.1423 -13.4727 -0.012>, rot=<-0.0092 -0.707 0.0092 -0.7071>}
 Backpack_BR {parent=attach_backpack, type=Null, loc=<0 0 0>, rot=<0 0 0 1>}
 attach_cape {parent=spine_05, loc=<9.0102 -11.713 -0.0122>, rot=<0.0617 -0.7044 -0.0617 -0.7044>}
 cape {parent=attach_cape, type=Null, loc=<0 0 0>, rot=<0 0 0 1>}
 deform_pec_aim_base_l {parent=spine_04, loc=<3.2456 9.3309 -1.2237>, rot=<0.0603 -0.4033 0.3318 -0.8507>}
 deform_pec_aim_l {parent=deform_pec_aim_base_l, loc=<0 0 0>, rot=<0 0 0 1>}
 deform_pec_base_l {parent=deform_pec_aim_l, loc=<9.3665 0 0>, rot=<0 0 0 1>}
 deform_pec_l {parent=deform_pec_base_l, loc=<0 0 0>, rot=<0 0 0 1>}
 deform_pec_aim_base_r {parent=spine_04, loc=<3.2452 9.3309 1.2237>, rot=<-0.4033 -0.0603 0.8507 0.3318>}
 deform_pec_aim_r {parent=deform_pec_aim_base_r, loc=<0 0 0>, rot=<0 0 0 1>}
 deform_pec_base_r {parent=deform_pec_aim_r, loc=<-9.3667 -0.0003 -0.0004>, rot=<0 0 0 1>}
 deform_pec_r {parent=deform_pec_base_r, loc=<0 0 0>, rot=<0 0 0 1>}
 base_bag_chest_l {parent=spine_04, loc=<3.3995 10.8666 -8.551>, rot=<-0.3924 -0.5042 -0.2732 -0.7192>}
 dyn_bag_chest_l {parent=base_bag_chest_l, loc=<0 0 0>, rot=<0 0 0 1>}
 base_bag_chest_r {parent=spine_04, loc=<3.3991 10.8666 8.551>, rot=<0.7192 0.2732 0.5042 0.3924>}
 dyn_bag_chest_r {parent=base_bag_chest_r, loc=<0 0 0>, rot=<0 0 0 1>}
 base_radio {parent=spine_04, loc=<2.648 7.7586 11.6349>, rot=<0.7023 0.0045 0.7096 0.0567>}
 dyn_radio {parent=base_radio, loc=<0 0 0>, rot=<0 0 0 1>}
 base_bag_spine_r {parent=spine_03, loc=<-0.8527 2.5415 11.2674>, rot=<0.5705 0.0313 0.8207 0.0051>}
 dyn_bag_spine_r {parent=base_bag_spine_r, loc=<0 0 0>, rot=<0 0 0 1>}
 base_bag_spine_l {parent=spine_03, loc=<-0.8531 2.5415 -11.2674>, rot=<-0.0051 -0.8207 -0.0313 -0.5705>}
 dyn_bag_spine_l {parent=base_bag_spine_l, loc=<0 0 0>, rot=<0 0 0 1>}
 for_aim_bullets_back_l {parent=spine_03, loc=<5.687 -6.858 -8.115>, rot=<0.5304 -0.7976 -0.1422 -0.2496>}
 base_bullets_back_l {parent=spine_03, loc=<0.8845 -4.588 -10.0131>, rot=<0.4326 -0.7708 0.0057 -0.4676>}
 dyn_bullets_back_l {parent=base_bullets_back_l, loc=<0 0 0>, rot=<0 0 0 1>}
 base_bullets_front_l {parent=spine_03, loc=<2.6022 9.4892 -10.7>, rot=<-0.4106 -0.7831 0.1084 -0.4544>}
 dyn_bullets_front_l {parent=base_bullets_front_l, loc=<0 0 0>, rot=<0 0 0 1>}
 base_bullets_front_r {parent=spine_03, loc=<2.6026 9.4892 10.7>, rot=<0.7831 -0.4106 -0.4544 -0.1084>}
 dyn_bullets_front_r {parent=base_bullets_front_r, loc=<0 0 0>, rot=<0 0 0 1>}
 base_bullets_back_r {parent=spine_03, loc=<0.884 -4.5881 10.0131>, rot=<0.7708 0.4326 -0.4676 -0.0057>}
 dyn_bullets_back_r {parent=base_bullets_back_r, loc=<0 0 0>, rot=<0 0 0 1>}
 thigh_l {parent=pelvis, loc=<-1.0855 -0.2109 -8.7053>, rot=<0.0009 0.0355 0.0155 -0.9993>}
 calf_l {parent=thigh_l, loc=<-42.5284 -0.0019 -0.7297>, rot=<0.0006 -0.0161 -0.0391 -0.9991>}
 foot_l {parent=calf_l, loc=<-46.7401 -2.5014 -0.592>, rot=<0.0427 -0.0238 0.0353 -0.9982>}
 ball_l {parent=foot_l, loc=<-3.5658 10.8066 -0.1544>, rot=<0.0001 -0.0001 -0.7063 -0.7079>}
 base_tongue_shoe_l {parent=foot_l, loc=<-0.444 6.0961 0.5751>, rot=<-0.3426 -0.6267 -0.5916 -0.3739>}
 dyn_tongue_shoe_l {parent=base_tongue_shoe_l, loc=<0 0 0>, rot=<0 0 0 1>}
 calf_twist_02_l {parent=calf_l, loc=<-19.1546 -1.7943 -0.2638>, rot=<0.0237 0.0095 -0.0035 -0.9997>}
 calf_twist_01_l {parent=calf_l, loc=<-35.8483 -1.9197 -0.5794>, rot=<0.0237 0.0095 -0.0035 -0.9997>}
 deform_knee_in_base_l {parent=calf_l, loc=<0.8031 -5.0231 0.2456>, rot=<0 0 0 -1>}
 deform_knee_in_l {parent=deform_knee_in_base_l, loc=<0 0 0>, rot=<0 0 0 1>}
 deform_knee_out_base_l {parent=calf_l, loc=<0 4 0>, rot=<0 0 0 -1>}
 deform_knee_out_l {parent=deform_knee_out_base_l, loc=<0 0 0>, rot=<0 0 0 1>}
 foot_l_null {parent=calf_l, type=Null, loc=<-46.7402 -2.5014 -0.592>, rot=<0.0427 -0.0238 0.0353 -0.9982>}
 thigh_twist_01_l {parent=thigh_l, loc=<-27.952 0 0>, rot=<0.0013 0 0 -1>}
 deform_groin_base_l {parent=thigh_l, loc=<0.4801 5.6776 -0.2486>, rot=<0 0 0 -1>}
 deform_groin_l {parent=deform_groin_base_l, loc=<0 0 0>, rot=<0 0 0 1>}
 rivet_for_aim_grenade_2 {parent=thigh_l, loc=<-23.798 -9.8095 -6.2524>, rot=<-0.0491 -0.8101 0.573 -0.114>}
 rivet_for_aim_grenade_1 {parent=thigh_l, loc=<-20.5564 4.5421 -8.6461>, rot=<0.0452 -0.8902 -0.4234 -0.1617>}
 thigh_r {parent=pelvis, loc=<-1.0855 -0.2109 8.7053>, rot=<0.0355 -0.0009 0.9993 0.0155>}
 calf_r {parent=thigh_r, loc=<42.5284 0.0019 0.7297>, rot=<0.0006 -0.0161 -0.0391 -0.9991>}
 foot_r {parent=calf_r, loc=<46.7401 2.5014 0.592>, rot=<0.0427 -0.0238 0.0353 -0.9982>}
 ball_r {parent=foot_r, loc=<3.5658 -10.8066 0.1544>, rot=<0.0001 -0.0001 -0.7063 -0.7079>}
 base_tongue_shoe_r {parent=foot_r, loc=<-0.533 -6.5717 -0.025>, rot=<-0.4905 0.5116 0.4936 -0.504>}
 dyn_tongue_shoe_r {parent=base_tongue_shoe_r, loc=<0 0 0>, rot=<0 0 0 1>}
 calf_twist_02_r {parent=calf_r, loc=<19.1546 1.7943 0.2638>, rot=<0.0237 0.0095 -0.0035 -0.9997>}
 calf_twist_01_r {parent=calf_r, loc=<35.8483 1.9197 0.5794>, rot=<0.0237 0.0095 -0.0035 -0.9997>}
 deform_knee_in_base_r {parent=calf_r, loc=<-0.8031 5.0231 -0.2457>, rot=<0 0 0 -1>}
 deform_knee_in_r {parent=deform_knee_in_base_r, loc=<0 0 0>, rot=<0 0 0 1>}
 deform_knee_out_base_r {parent=calf_r, loc=<0 -4 0>, rot=<0 0 0 -1>}
 deform_knee_out_r {parent=deform_knee_out_base_r, loc=<0 0 0>, rot=<0 0 0 1>}
 foot_r_null {parent=calf_r, type=Null, loc=<46.7401 2.5014 0.592>, rot=<0.0427 -0.0238 0.0353 -0.9982>}
 calf_counter_r {parent=calf_r, loc=<0 0 0>, rot=<0 0 0 1>}
 thigh_twist_01_r {parent=thigh_r, loc=<27.952 0 0>, rot=<0.0013 0 0 -1>}
 deform_groin_base_r {parent=thigh_r, loc=<-0.4801 -5.6776 0.2486>, rot=<0 0 0 -1>}
 deform_groin_r {parent=deform_groin_base_r, loc=<0 0 0>, rot=<0 0 0 1>}
 thigh_counter_r {parent=thigh_r, loc=<42.5284 0.0019 0.7297>, rot=<0 0 0 1>}
 deform_glute_base_l {parent=pelvis, loc=<-7.3466 -9.0213 -9.1245>, rot=<0.0172 0.031 -0.4477 -0.8935>}
 deform_glute_l {parent=deform_glute_base_l, loc=<0 0 0>, rot=<0 0 0 1>}
 deform_glute_base_r {parent=pelvis, loc=<-7.3464 -9.0213 9.1245>, rot=<0.031 -0.0172 0.8935 -0.4477>}
 deform_glute_r {parent=deform_glute_base_r, loc=<0 0 0>, rot=<0 0 0 1>}
 base_strap_back_0 {parent=pelvis, loc=<4.4821 -8.6261 -6.3198>, rot=<-0.2208 0.5338 0.8147 -0.0508>}
 dyn_strap_back_0 {parent=base_strap_back_0, loc=<0 0 0>, rot=<0 0 0 1>}
 dyn_strap_back_1 {parent=dyn_strap_back_0, loc=<9.1606 0 0>, rot=<0.0001 0.014 0.0038 -0.9999>}
 dyn_strap_back_2 {parent=dyn_strap_back_1, loc=<4.9621 0 0>, rot=<0.0123 0.2311 0.0517 -0.9715>}
 dyn_strap_back_3 {parent=dyn_strap_back_2, loc=<7.3967 0 0>, rot=<-0.0031 -0.0245 0.1239 -0.992>}
 dyn_strap_back_4 {parent=dyn_strap_back_3, loc=<10.6214 0 0>, rot=<0.0013 0.0299 0.0418 -0.9987>}
 dyn_strap_back_5 {parent=dyn_strap_back_4, loc=<6.9558 0 0>, rot=<0 0 0 1>}
 base_grenade_2 {parent=dyn_strap_back_2, loc=<5.4489 0.6139 0.0821>, rot=<0.9916 -0.0131 -0.1088 -0.0681>}
 dyn_grenade_2 {parent=base_grenade_2, loc=<0 0 0>, rot=<0 0 0 1>}
 base_strap_front_helper {parent=pelvis, loc=<4.2772 4.7822 -10.4325>, rot=<-0.5089 -0.491 -0.491 -0.5089>}
 base_strap_front_helper_orient {parent=base_strap_front_helper, loc=<0 0 0>, rot=<0 0 0 1>}
 base_strap_front_helper_auto {parent=base_strap_front_helper_orient, loc=<0 0 0>, rot=<0 0 0 1>}
 base_strap_front_helper_offset {parent=base_strap_front_helper_auto, loc=<0 0 0>, rot=<0 0 0 1>}
 base_strap_front_0 {parent=base_strap_front_helper_offset, loc=<0 0 0>, rot=<-0.1203 -0.5204 0.3822 -0.7541>}
 dyn_strap_front_0 {parent=base_strap_front_0, loc=<0 0 0>, rot=<0 0 0 1>}
 dyn_strap_front_1 {parent=dyn_strap_front_0, loc=<7.1759 0 0>, rot=<0 -0.1106 0 -0.9939>}
 dyn_strap_front_2 {parent=dyn_strap_front_1, loc=<4.8 0 0>, rot=<0.0001 -0.0117 -0.0063 -0.9999>}
 dyn_strap_front_3 {parent=dyn_strap_front_2, loc=<7.6825 0 0>, rot=<0.0609 -0.1343 0.0577 -0.9874>}
 dyn_strap_front_4 {parent=dyn_strap_front_3, loc=<7.307 0 0>, rot=<-0.0005 -0.0172 0.0265 -0.9995>}
 dyn_strap_front_5 {parent=dyn_strap_front_4, loc=<8.0876 0 0>, rot=<-0.0216 -0.1617 0.1276 -0.9783>}
 base_grenade_1 {parent=dyn_strap_front_2, loc=<5.774 0.0069 -0.3652>, rot=<0.0446 0.0556 0.0245 -0.9972>}
 dyn_grenade_1 {parent=base_grenade_1, loc=<0 0 0>, rot=<0 0 0 1>}
 dyn_grenade_1_offset {parent=dyn_grenade_1, loc=<5.6624 -0.2647 -1.2595>, rot=<0 0 0 1>}
 thigh_l_get_rot_off {parent=pelvis, loc=<-1.0855 -0.2109 -8.7053>, rot=<0.0009 0.0355 0.0155 -0.9992>}
 thigh_l_get_rot {parent=thigh_l_get_rot_off, loc=<0 0 0>, rot=<0 0 0 1>}
 thigh_l_set_rot_front_off {parent=pelvis, loc=<-1.0855 -0.2109 -8.7053>, rot=<0.0009 0.0355 0.0155 -0.9992>}
 thigh_l_set_rot_front {parent=thigh_l_set_rot_front_off, loc=<0 0 0>, rot=<0 0 0 1>}
 thigh_l_set_rot_back_off {parent=pelvis, loc=<-1.0855 -0.2109 -8.7053>, rot=<0.0009 0.0355 0.0155 -0.9992>}
 thigh_l_set_rot_back {parent=thigh_l_set_rot_back_off, loc=<0 0 0>, rot=<0 0 0 1>}
 for_aim_jnt_3 {parent=root, loc=<-5.983 -8.988 119.836>, rot=<-0.448 -0.4243 0.62 -0.4846>}
 for_aim_jnt_1 {parent=root, loc=<-9.381 -12.899 133.939>, rot=<-0.3457 -0.3293 0.8197 -0.3166>}
 thigh_l_ctrl {controlType=EULER_TRANSFORM, type=Control, loc=<0 0 0>, rot=<0 0 0 1>}
 foot_l_ctrl {controlType=EULER_TRANSFORM, parent=thigh_l_ctrl, type=Control, loc=<0 0 0>, rot=<0 0 0 1>}
 foot_r_ctrl {controlType=EULER_TRANSFORM, type=Control, loc=<0 0 0>, rot=<0 0 0 1>}
 dyn_strap_front_3_ctrl {controlType=EULER_TRANSFORM, type=Control, loc=<0 0 0>, rot=<0 0 0 1>}
]
