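# Physics asset for the female medium RaggedRebel player body cosmetic, setting up a dangling strap with front and back bone chains.
# Each chain runs six capsule bodies from a small base anchor, linked by joints with tight swing and twist limits.
# The two chain tips share an angularly locked constraint, closing the strap into a loop.
# Capsules on pelvis, left thigh, left calf and a large root body provide body shapes around the strap.
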PhysicsAsset /BRCosmetics/Characters/Player/Female/Medium/Bodies/F_MED_RaggedRebel/Meshes/F_MED_RaggedRebel_Physics
bodies {
 base_strap_back_0 [{shapeType=Capsule, length=0.505, radius=0.505, shapeName=base_strap_back_0_capsule, rotation=<rot 0 -90 0>}]
 base_strap_front_0 [{shapeType=Capsule, length=0.505, radius=0.505, shapeName=base_strap_front_0_capsule, rotation=<rot 0 -90 0>}]
 calf_l [{shapeType=Capsule, length=49.258, radius=8.742, shapeName=calf_l_capsule, center=<-24.4522 -2.8096 0>, rotation=<rot 90 0 0>}]
 dyn_strap_back_0 [{
   shapeType=Capsule
   length=7.7315
   radius=1.9239
   shapeName=dyn_strap_back_0_capsule
   center=<4.9603 0 -1.8202>
   rotation=<rot 84.2415 179.9997 -179.9999>
  }]
 dyn_strap_back_1 [{
   shapeType=Capsule
   length=8.4853
   radius=1.5191
   shapeName=dyn_strap_back_1_capsule
   center=<2 0 -1>
   rotation=<rot 77.1629 -179.9952 -179.9933>
  }]
 dyn_strap_back_2 [{shapeType=Capsule, length=7.1997, radius=2.7003, shapeName=dyn_strap_back_2_capsule, center=<2 0 -1.6841>, rotation=<rot 90 0 0>}]
 dyn_strap_back_3 [{shapeType=Capsule, length=8.3, radius=1.6, shapeName=dyn_strap_back_3_capsule, center=<3 0 0>, rotation=<rot 90 0 0>}]
 dyn_strap_back_4 [{shapeType=Capsule, length=8.3, radius=1.6, shapeName=dyn_strap_back_4_capsule, center=<2 0 0>, rotation=<rot 90 0 0>}]
 dyn_strap_back_5 [{shapeType=Capsule, length=8.3, radius=1.6, shapeName=dyn_strap_back_5_capsule, center=<2 0 0>, rotation=<rot -49.5528 -20.1667 -17.9063>}]
 dyn_strap_front_0 [{
   shapeType=Capsule
   length=7.6836
   radius=1.7432
   shapeName=dyn_strap_front_0_capsule
   center=<2.641 0 3.2166>
   rotation=<rot 65.9865 0 -0.0001>
  }]
 dyn_strap_front_1 [{
   shapeType=Capsule
   length=8.1029
   radius=1.7971
   shapeName=dyn_strap_front_1_capsule
   center=<2.0207 0 0.8212>
   rotation=<rot 86.3131 -0.0002 -0.0002>
  }]
 dyn_strap_front_2 [
  {
   shapeType=Capsule
   length=14.4061
   radius=2.3964
   shapeName=dyn_strap_front_2_capsule
   center=<3.0672 0 1.9346>
   rotation=<rot 86.7366 -0.0012 -0.0013>
  }
  {
   shapeType=Capsule
   length=8.6532
   radius=3.7549
   shapeName=dyn_strap_front_2_capsule
   center=<3.1607 0 2.1439>
   rotation=<rot 83.5318 -0.0002 87.5679>
  }
 ]
 dyn_strap_front_3 [{
   shapeType=Capsule
   length=12.6359
   radius=3.1314
   shapeName=dyn_strap_front_3_capsule
   center=<1.9289 0 0.61>
   rotation=<rot 83.3546 -179.9998 -179.9999>
  }]
 dyn_strap_front_4 [{
   shapeType=Capsule
   length=12.6907
   radius=2.4693
   shapeName=dyn_strap_front_4_capsule
   center=<1.932 0 0.5928>
   rotation=<rot 83.456 -179.9998 -179.9999>
  }]
 dyn_strap_front_5 [{shapeType=Capsule, length=8.3, radius=1.6, shapeName=dyn_strap_front_5_capsule, center=<2 0 0>, rotation=<rot 90 0 0>}]
 pelvis [
  {shapeType=Capsule, length=19.798, radius=9.1089, shapeName=pelvis_capsule, center=<-7.8979 -2.5716 0>}
  {shapeType=Capsule, length=10.4439, radius=9.2683, shapeName=pelvis_capsule, center=<-3.7879 -1.3085 0>}
 ]
 root [{shapeType=Capsule, length=0, radius=70, shapeName=root_capsule, center=<0 0 -56>}]
 thigh_l [
  {
   shapeType=Capsule
   length=36.2888
   radius=8.559
   shapeName=thigh_l_capsule
   center=<-20.059 -1.682 0.1163>
   rotation=<rot 89.8509 152.3246 152.452>
  }
  {
   shapeType=Capsule
   length=27.3763
   radius=8.559
   shapeName=thigh_l_capsule
   center=<-19.1441 -0.092 0.1394>
   rotation=<rot 89.8509 152.3241 151.5002>
  }
 ]
}
constraints [
 {bone1Name=dyn_strap_back_1, bone2Name=dyn_strap_back_0, swing1LimitDegrees=20, swing2LimitDegrees=20, twistLimitDegrees=2}
 {bone1Name=dyn_strap_back_2, bone2Name=dyn_strap_back_1, swing1LimitDegrees=20, swing2LimitDegrees=20, twistLimitDegrees=2}
 {bone1Name=dyn_strap_back_3, bone2Name=dyn_strap_back_2, swing1LimitDegrees=20, swing2LimitDegrees=20, twistLimitDegrees=2}
 {bone1Name=dyn_strap_back_4, bone2Name=dyn_strap_back_3, swing1LimitDegrees=20, swing2LimitDegrees=20, twistLimitDegrees=2}
 {bone1Name=dyn_strap_back_5, bone2Name=dyn_strap_back_4, swing1LimitDegrees=20, swing2LimitDegrees=20, twistLimitDegrees=2}
 {bone1Name=dyn_strap_front_1, bone2Name=dyn_strap_front_0, swing1LimitDegrees=20, swing2LimitDegrees=20, twistLimitDegrees=2}
 {bone1Name=dyn_strap_front_2, bone2Name=dyn_strap_front_1, swing1LimitDegrees=20, swing2LimitDegrees=20, twistLimitDegrees=2}
 {bone1Name=dyn_strap_front_3, bone2Name=dyn_strap_front_2, swing1LimitDegrees=20, swing2LimitDegrees=20, twistLimitDegrees=2}
 {bone1Name=dyn_strap_front_4, bone2Name=dyn_strap_front_3, swing1LimitDegrees=20, swing2LimitDegrees=20, twistLimitDegrees=2}
 {bone1Name=dyn_strap_front_5, bone2Name=dyn_strap_front_4, swing1LimitDegrees=20, swing2LimitDegrees=20, twistLimitDegrees=2}
 {bone1Name=dyn_strap_back_0, bone2Name=base_strap_back_0, swing1LimitDegrees=20, swing2LimitDegrees=20, twistLimitDegrees=2}
 {bone1Name=dyn_strap_front_0, bone2Name=base_strap_front_0, swing1LimitDegrees=20, swing2LimitDegrees=20, twistLimitDegrees=2}
 {
  bone1Name=dyn_strap_back_5
  bone2Name=dyn_strap_front_5
  swing1LimitDegrees=45
  swing1Motion=Locked
  swing2LimitDegrees=45
  swing2Motion=Locked
  twistLimitDegrees=45
  twistMotion=Locked
 }
]
