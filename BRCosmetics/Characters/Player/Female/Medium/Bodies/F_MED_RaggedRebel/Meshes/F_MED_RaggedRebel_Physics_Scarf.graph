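# Physics asset for the scarf on the F_MED_RaggedRebel female medium player body, defining capsule bodies for secondary scarf motion.
# A six-bone dyn_scarf chain is linked by five constraints with locked twist and swing limits that widen toward the tip.
# Capsules on the upper spine, right clavicle and right upper arm supply body shapes around the scarf.
# A tiny base_scarf_1 capsule has no constraint to the chain.
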PhysicsAsset /BRCosmetics/Characters/Player/Female/Medium/Bodies/F_MED_RaggedRebel/Meshes/F_MED_RaggedRebel_Physics_Scarf
bodies {
 base_scarf_1 [{shapeType=Capsule, length=0.505, radius=0.505, shapeName=base_scarf_1_capsule, rotation=<rot 0 -90 0>}]
 clavicle_r [{
   shapeType=Capsule
   length=7.2211
   radius=6.5641
   shapeName=clavicle_r_capsule
   center=<-4.4627 1.8376 0.0364>
   rotation=<rot 79.9998 -0.026 -7.7937>
  }]
 dyn_scarf_1 [{shapeType=Capsule, length=8, radius=2, shapeName=dyn_scarf_1_capsule, center=<2 0 1>, rotation=<rot 90 0 0>}]
 dyn_scarf_2 [{shapeType=Capsule, length=8.6501, radius=1.3499, shapeName=dyn_scarf_2_capsule, center=<2 0 2>, rotation=<rot 90 0 0>}]
 dyn_scarf_3 [{shapeType=Capsule, length=8.6501, radius=1.3499, shapeName=dyn_scarf_3_capsule, center=<4.9317 0 1>, rotation=<rot 90 0 0>}]
 dyn_scarf_4 [{shapeType=Capsule, length=8.6501, radius=1.3499, shapeName=dyn_scarf_4_capsule, center=<2 0 1>, rotation=<rot 90 0 0>}]
 dyn_scarf_5 [{shapeType=Capsule, length=8.6501, radius=1.3499, shapeName=dyn_scarf_5_capsule, center=<2 0 1>, rotation=<rot 90 0 0>}]
 dyn_scarf_6 [{shapeType=Capsule, length=8.6501, radius=1.3499, shapeName=dyn_scarf_6_capsule, center=<2 0 1>, rotation=<rot 90 0 0>}]
 spine_04 [
  {shapeType=Capsule, length=8.1119, radius=10.3882, shapeName=spine_04_capsule, center=<-9.6845 -0.0111 1.0821>}
  {
   shapeType=Capsule
   length=24.7294
   radius=9.2705
   shapeName=spine_04_capsule
   center=<0.7926 -1.9793 0.1439>
   rotation=<rot 0 -0.0002 7.6382>
  }
 ]
 spine_05 [{shapeType=Capsule, length=9.8171, radius=10.458, shapeName=spine_05_capsule, center=<0.8123 0.3089 0>}]
 upperarm_r [{
   shapeType=Capsule
   length=19.9615
   radius=4.7549
   shapeName=upperarm_r_capsule
   center=<-9.9691 0.6879 0.0876>
   rotation=<rot 87.2717 -179.9992 -179.9993>
  }]
}
constraints [
 {bone1Name=dyn_scarf_2, bone2Name=dyn_scarf_1, swing1LimitDegrees=0, swing2LimitDegrees=8, twistLimitDegrees=0}
 {bone1Name=dyn_scarf_3, bone2Name=dyn_scarf_2, swing1LimitDegrees=0, swing2LimitDegrees=12, twistLimitDegrees=0}
 {bone1Name=dyn_scarf_4, bone2Name=dyn_scarf_3, swing1LimitDegrees=5, swing2LimitDegrees=15, twistLimitDegrees=0}
 {bone1Name=dyn_scarf_5, bone2Name=dyn_scarf_4, swing1LimitDegrees=15, swing2LimitDegrees=30, twistLimitDegrees=0}
 {bone1Name=dyn_scarf_6, bone2Name=dyn_scarf_5, swing1LimitDegrees=15, swing2LimitDegrees=30, twistLimitDegrees=0}
]
