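# Post-copy AnimBlueprint for the F_MED_RaggedRebel cosmetic body, adding secondary motion to straps, scarf, bags, radio, grenades and shoe tongues.
# Cached stages run parent pose copy, Control Rig, additive bone offsets, AnimDynamics, then two RigidBody simulations.
# Grenades aim at weighted rivet bones before their own AnimDynamics; weighted constraints then tie scarf bones to the body.
# Update event derives gravity, simulation settings and corrective values from movement, crouch, head angle, backbling and front-end state.
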AnimBlueprint /BRCosmetics/Characters/Player/Female/Medium/Bodies/F_MED_RaggedRebel/Meshes/F_MED_RaggedRebel_AnimBP
target_skeleton=&/F_MED_RaggedRebel_Skeleton
graphs [
 AnimGraph {nodes [
   Output Pose(Result=Use cached pose 'CapeAndBackpackCache'()) {layerGroup=None}
   Save cached pose 'InPoseCache'(Pose=Copy Pose From Mesh() {bUseAttachedParent=true}) {cacheName=InPoseCache}
   n1 = Control Rig(Source=Use cached pose 'InPoseCache'()) {
    alphaBoolBlend {blendOption=Linear}
    controlRigAssetReference {blueprintRigClass=&/F_MED_RaggedRebel_CtrlRig.F_MED_RaggedRebel_CtrlRig_C}
   }
   Save cached pose 'ControlRigCache'(Pose=@n1) {cacheName=ControlRigCache}
   n2 = Anim Dynamics - Bone: dyn_arm_strap_3_l!(SimulationSpace:AnimPhysSimSpaceType Enum, ComponentPose=Local To Component(LocalPose=Use cached pose 'modifyTransformCache'()), Alpha=1) -> (Pose) {
    angularDampingOverride=1
    angularSpringConstant=8000
    bAngularSpring=true
    bOverrideAngularDamping=true
    bUseGravityOverride=true
    gravityScale=0.35
    simulationSpace=RootRelative
    alphaBoolBlend {blendOption=Linear}
    boundBone {boneName=dyn_arm_strap_3_l}
    componentAppliedLinearAccClamp=<5000 5000 5000>
    componentLinearAccScale=<0.6 0.6 0.6>
    componentLinearVelScale=<0.6 0.6 0.6>
    gravityOverride {z=-980}
    physicsBodyDefinitions [{
      boundBone {boneName=dyn_arm_strap_3_l}
      boxExtents=<35 35 35>
      constraintSetup {angularLimitsMax=<0 0 5>, angularLimitsMin=<0 0 -40>, angularTarget=<1 0 0>}
      localJointOffset=<10 0 0>
     }]
   }
   n3 = Anim Dynamics - Bone: dyn_arm_strap_2_l!(SimulationSpace:AnimPhysSimSpaceType Enum, ComponentPose=@n2, Alpha=1) -> (Pose) {
    angularDampingOverride=1
    angularSpringConstant=8000
    bAngularSpring=true
    bOverrideAngularDamping=true
    bUseGravityOverride=true
    gravityScale=0.35
    simulationSpace=RootRelative
    alphaBoolBlend {blendOption=Linear}
    boundBone {boneName=dyn_arm_strap_2_l}
    componentAppliedLinearAccClamp=<5000 5000 5000>
    componentLinearAccScale=<0.6 0.6 0.6>
    componentLinearVelScale=<0.6 0.6 0.6>
    gravityOverride {z=-980}
    physicsBodyDefinitions [{
      boundBone {boneName=dyn_arm_strap_2_l}
      boxExtents=<30 30 30>
      constraintSetup {angularLimitsMax=<0 0 5>, angularLimitsMin=<0 0 -65>, angularTarget=<1 0 0>}
      localJointOffset=<10 0 0>
     }]
   }
   n4 = Anim Dynamics - Bone: dyn_arm_strap_1_l!(SimulationSpace:AnimPhysSimSpaceType Enum, ComponentPose=@n3, Alpha=1) -> (Pose) {
    angularDampingOverride=1
    angularSpringConstant=8000
    bAngularSpring=true
    bOverrideAngularDamping=true
    bUseGravityOverride=true
    gravityScale=0.35
    simulationSpace=RootRelative
    alphaBoolBlend {blendOption=Linear}
    boundBone {boneName=dyn_arm_strap_1_l}
    componentAppliedLinearAccClamp=<5000 5000 5000>
    componentLinearAccScale=<0.6 0.6 0.6>
    componentLinearVelScale=<0.6 0.6 0.6>
    gravityOverride {z=-980}
    physicsBodyDefinitions [{
      boundBone {boneName=dyn_arm_strap_1_l}
      boxExtents=<30 30 30>
      constraintSetup {angularLimitsMax=<0 0 5>, angularLimitsMin=<0 0 -70>, angularTarget=<1 0 0>}
      localJointOffset=<10 0 0>
     }]
   }
   n5 = Anim Dynamics - Bone: dyn_arm_strap_3_r!(SimulationSpace:AnimPhysSimSpaceType Enum, ComponentPose=@n4, Alpha=1) -> (Pose) {
    angularDampingOverride=1
    angularSpringConstant=8000
    bAngularSpring=true
    bOverrideAngularDamping=true
    bUseGravityOverride=true
    gravityScale=0.35
    simulationSpace=RootRelative
    alphaBoolBlend {blendOption=Linear}
    boundBone {boneName=dyn_arm_strap_3_r}
    componentAppliedLinearAccClamp=<5000 5000 5000>
    componentLinearAccScale=<0.6 0.6 0.6>
    componentLinearVelScale=<0.6 0.6 0.6>
    gravityOverride {z=-980}
    physicsBodyDefinitions [{
      boundBone {boneName=dyn_arm_strap_3_r}
      boxExtents=<30 30 30>
      constraintSetup {angularLimitsMax=<0 0 5>, angularLimitsMin=<0 0 -55>, angularTarget=<1 0 0>}
      localJointOffset=<10 0 0>
     }]
   }
   n6 = Anim Dynamics - Bone: dyn_arm_strap_2_r!(SimulationSpace:AnimPhysSimSpaceType Enum, ComponentPose=@n5, Alpha=1) -> (Pose) {
    angularDampingOverride=1
    angularSpringConstant=8000
    bAngularSpring=true
    bOverrideAngularDamping=true
    bUseGravityOverride=true
    gravityScale=0.35
    simulationSpace=RootRelative
    alphaBoolBlend {blendOption=Linear}
    boundBone {boneName=dyn_arm_strap_2_r}
    componentAppliedLinearAccClamp=<5000 5000 5000>
    componentLinearAccScale=<0.6 0.6 0.6>
    componentLinearVelScale=<0.6 0.6 0.6>
    gravityOverride {z=-980}
    physicsBodyDefinitions [{
      boundBone {boneName=dyn_arm_strap_2_r}
      boxExtents=<35 35 35>
      constraintSetup {angularLimitsMax=<0 0 5>, angularLimitsMin=<0 0 -65>, angularTarget=<1 0 0>}
      localJointOffset=<10 0 0>
     }]
   }
   n7 = Anim Dynamics - Bone: dyn_arm_strap_1_r!(SimulationSpace:AnimPhysSimSpaceType Enum, ComponentPose=@n6, Alpha=1) -> (Pose) {
    angularDampingOverride=1
    angularSpringConstant=8000
    bAngularSpring=true
    bOverrideAngularDamping=true
    bUseGravityOverride=true
    gravityScale=0.35
    simulationSpace=RootRelative
    alphaBoolBlend {blendOption=Linear}
    boundBone {boneName=dyn_arm_strap_1_r}
    componentAppliedLinearAccClamp=<5000 5000 5000>
    componentLinearAccScale=<0.6 0.6 0.6>
    componentLinearVelScale=<0.6 0.6 0.6>
    gravityOverride {z=-980}
    physicsBodyDefinitions [{
      boundBone {boneName=dyn_arm_strap_1_r}
      boxExtents=<30 30 30>
      constraintSetup {angularLimitsMax=<0 0 5>, angularLimitsMin=<0 0 -75>, angularTarget=<1 0 0>}
      localJointOffset=<10 0 0>
     }]
   }
   n8 = Anim Dynamics - Bone: dyn_bag_chest_r!(ComponentPose=@n7) -> (Pose) {
    angularDampingOverride=1
    angularSpringConstant=20000
    bAngularSpring=true
    bOverrideAngularDamping=true
    bUseGravityOverride=true
    gravityScale=0.35
    simulationSpace=RootRelative
    alphaBoolBlend {blendOption=Linear}
    boundBone {boneName=dyn_bag_chest_r}
    componentAppliedLinearAccClamp=<5000 5000 5000>
    componentLinearAccScale=<0.6 0.6 0.6>
    componentLinearVelScale=<0.6 0.6 0.6>
    gravityOverride {z=-980}
    physicsBodyDefinitions [{
      boundBone {boneName=dyn_bag_chest_r}
      boxExtents=<45 45 45>
      constraintSetup {angularLimitsMax=<0 5 10>, angularLimitsMin=<0 -10 -10>, angularTarget=<1 0 0>}
      localJointOffset=<4 0 0>
     }]
   }
   n9 = Anim Dynamics - Bone: dyn_radio!(ComponentPose=@n8) -> (Pose) {
    angularDampingOverride=1
    angularSpringConstant=20000
    bAngularSpring=true
    bOverrideAngularDamping=true
    bUseGravityOverride=true
    gravityScale=0.35
    simulationSpace=RootRelative
    alphaBoolBlend {blendOption=Linear}
    boundBone {boneName=dyn_radio}
    componentAppliedLinearAccClamp=<5000 5000 5000>
    componentLinearAccScale=<0.6 0.6 0.6>
    componentLinearVelScale=<0.6 0.6 0.6>
    gravityOverride {z=-980}
    physicsBodyDefinitions [{
      boundBone {boneName=dyn_radio}
      boxExtents=<55 55 55>
      constraintSetup {angularLimitsMax=<0 2 5>, angularLimitsMin=<0 -5 -5>, angularTarget=<1 0 0>}
      localJointOffset=<4 0 0>
     }]
   }
   n10 = Anim Dynamics - Bone: dyn_bag_chest_l!(ComponentPose=@n9) -> (Pose) {
    angularDampingOverride=1
    angularSpringConstant=20000
    bAngularSpring=true
    bOverrideAngularDamping=true
    bUseGravityOverride=true
    gravityScale=0.35
    simulationSpace=RootRelative
    alphaBoolBlend {blendOption=Linear}
    boundBone {boneName=dyn_bag_chest_l}
    componentAppliedLinearAccClamp=<5000 5000 5000>
    componentLinearAccScale=<0.6 0.6 0.6>
    componentLinearVelScale=<0.6 0.6 0.6>
    gravityOverride {z=-980}
    physicsBodyDefinitions [{
      boundBone {boneName=dyn_bag_chest_l}
      boxExtents=<45 45 45>
      constraintSetup {angularLimitsMax=<0 5 10>, angularLimitsMin=<0 -10 -10>, angularTarget=<1 0 0>}
      localJointOffset=<4 0 0>
     }]
   }
   n11 = Anim Dynamics - Bone: dyn_bag_spine_l!(ComponentPose=@n10) -> (Pose) {
    angularDampingOverride=1
    angularSpringConstant=20000
    bAngularSpring=true
    bOverrideAngularDamping=true
    bUseGravityOverride=true
    gravityScale=0.35
    simulationSpace=RootRelative
    alphaBoolBlend {blendOption=Linear}
    boundBone {boneName=dyn_bag_spine_l}
    componentAppliedLinearAccClamp=<5000 5000 5000>
    componentLinearAccScale=<0.6 0.6 0.6>
    componentLinearVelScale=<0.6 0.6 0.6>
    gravityOverride {z=-980}
    physicsBodyDefinitions [{
      boundBone {boneName=dyn_bag_spine_l}
      boxExtents=<45 45 45>
      constraintSetup {angularLimitsMax=<0 10 10>, angularLimitsMin=<0 -10 -10>, angularTarget=<1 0 0>}
      localJointOffset=<4 0 0>
     }]
   }
   n12 = Anim Dynamics - Bone: dyn_bag_spine_r!(ComponentPose=@n11) -> (Pose) {
    angularDampingOverride=1
    angularSpringConstant=20000
    bAngularSpring=true
    bOverrideAngularDamping=true
    bUseGravityOverride=true
    gravityScale=0.35
    simulationSpace=RootRelative
    alphaBoolBlend {blendOption=Linear}
    boundBone {boneName=dyn_bag_spine_r}
    componentAppliedLinearAccClamp=<5000 5000 5000>
    componentLinearAccScale=<0.6 0.6 0.6>
    componentLinearVelScale=<0.6 0.6 0.6>
    gravityOverride {z=-980}
    physicsBodyDefinitions [{
      boundBone {boneName=dyn_bag_spine_r}
      boxExtents=<45 45 45>
      constraintSetup {angularLimitsMax=<0 10 10>, angularLimitsMin=<0 -10 -10>, angularTarget=<1 0 0>}
      localJointOffset=<4 0 0>
     }]
   }
   n13 = Anim Dynamics - Bone: dyn_tongue_shoe_l!(SimulationSpace:AnimPhysSimSpaceType Enum, ComponentPose=@n12, Alpha=1) -> (Pose) {
    angularDampingOverride=1
    angularSpringConstant=12000
    bAngularSpring=true
    bOverrideAngularDamping=true
    bUseGravityOverride=true
    gravityScale=0.35
    simulationSpace=RootRelative
    alphaBoolBlend {blendOption=Linear}
    boundBone {boneName=dyn_tongue_shoe_l}
    componentAppliedLinearAccClamp=<5000 5000 5000>
    componentLinearAccScale=<0.6 0.6 0.6>
    componentLinearVelScale=<0.6 0.6 0.6>
    gravityOverride {z=-980}
    physicsBodyDefinitions [{
      boundBone {boneName=dyn_tongue_shoe_l}
      boxExtents=<20 20 20>
      constraintSetup {angularLimitsMax=<0 30 0>, angularLimitsMin=<0 -30 0>, angularTarget=<1 0 0>}
      localJointOffset=<6 0 0>
     }]
   }
   n14 = Anim Dynamics - Bone: dyn_tongue_shoe_r!(SimulationSpace:AnimPhysSimSpaceType Enum, ComponentPose=@n13, Alpha=1) -> (Pose) {
    angularDampingOverride=1
    angularSpringConstant=12000
    bAngularSpring=true
    bOverrideAngularDamping=true
    bUseGravityOverride=true
    gravityScale=0.35
    simulationSpace=RootRelative
    alphaBoolBlend {blendOption=Linear}
    boundBone {boneName=dyn_tongue_shoe_r}
    componentAppliedLinearAccClamp=<5000 5000 5000>
    componentLinearAccScale=<0.6 0.6 0.6>
    componentLinearVelScale=<0.6 0.6 0.6>
    gravityOverride {z=-980}
    physicsBodyDefinitions [{
      boundBone {boneName=dyn_tongue_shoe_r}
      boxExtents=<20 20 20>
      constraintSetup {angularLimitsMax=<0 30 0>, angularLimitsMin=<0 -15 0>, angularTarget=<1 0 0>}
      localJointOffset=<6 0 0>
     }]
   }
   Save cached pose 'animDynCache'(Pose=Component To Local(ComponentPose=@n14)) {cacheName=animDynCache}
   n15 = "Transform (Modify) Bone - Bone: base_bag_chest_r"!(ComponentPose=Local To Component(LocalPose=Use cached pose 'ControlRigCache'()), Alpha=1, Translation="0,-1,0") -> (Pose) {
    translationMode=BMM_Additive
    translationSpace=BCS_BoneSpace
    alphaBoolBlend {blendOption=Linear}
    boneToModify {boneName=base_bag_chest_r}
    translation {y=-1}
   }
   n16 = "Transform (Modify) Bone - Bone: base_strap_back_0"!(ComponentPose=@n15, bAlphaBoolEnabled:bool=True, Translation="0,0,-0.5", Rotation="22,30,14.7") -> (Pose) {
    alphaInputType=Bool
    rotationMode=BMM_Additive
    rotationSpace=BCS_BoneSpace
    translationMode=BMM_Additive
    translationSpace=BCS_BoneSpace
    alphaBoolBlend {blendInTime=0.2, blendOption=Linear, blendOutTime=0.2}
    boneToModify {boneName=base_strap_back_0}
    rotation=<rot 22 30 14.7>
    translation {z=-0.5}
   }
   n17 = "Transform (Modify) Bone - Bone: base_strap_back_0"(ComponentPose=@n16, bAlphaBoolEnabled=True, Translation="0,0,0.5", Rotation="-15,0,-14.7") {
    alphaInputType=Bool
    rotationMode=BMM_Additive
    rotationSpace=BCS_BoneSpace
    translationMode=BMM_Additive
    translationSpace=BCS_BoneSpace
    alphaBoolBlend {blendInTime=0.2, blendOption=Linear, blendOutTime=0.2}
    boneToModify {boneName=base_strap_back_0}
    rotation {pitch=-15, roll=-14.7}
    translation {z=0.5}
   }
   n18 = "Transform (Modify) Bone - Bone: base_strap_front_0"!(ComponentPose=@n17, bAlphaBoolEnabled:bool=True, Translation="3.025,-1.510,-0.940", Rotation="-9.266,-38.552,-7.278") -> (Pose) {
    alphaInputType=Bool
    rotationMode=BMM_Additive
    rotationSpace=BCS_BoneSpace
    translationMode=BMM_Additive
    translationSpace=BCS_BoneSpace
    alphaBoolBlend {blendInTime=0.2, blendOption=Linear, blendOutTime=0.2}
    boneToModify {boneName=base_strap_front_0}
    rotation=<rot -9.266 -38.552 -7.278>
    translation=<3.025 -1.51 -0.94>
   }
   n19 = "Transform (Modify) Bone - Bone: dyn_strap_front_1"!(ComponentPose=@n18, bAlphaBoolEnabled:bool=True, Translation="0,0,0", Rotation="-30,0,0") -> (Pose) {
    alphaInputType=Bool
    rotationMode=BMM_Additive
    rotationSpace=BCS_BoneSpace
    translationMode=BMM_Additive
    translationSpace=BCS_BoneSpace
    alphaBoolBlend {blendInTime=0.5, blendOption=Linear, blendOutTime=0.5}
    boneToModify {boneName=dyn_strap_front_1}
    rotation {pitch=-30}
   }
   n20 = "Transform (Modify) Bone - Bone: dyn_scarf_1"!(ComponentPose=@n19, bAlphaBoolEnabled:bool=True, Translation="0,0,2", Rotation="-20,0,0") -> (Pose) {
    alphaInputType=Bool
    rotationMode=BMM_Additive
    rotationSpace=BCS_BoneSpace
    translationMode=BMM_Additive
    alphaBoolBlend {blendInTime=1, blendOption=Linear, blendOutTime=1}
    boneToModify {boneName=dyn_scarf_1}
    rotation {pitch=-20}
    translation {z=2}
   }
   n21 = "Transform (Modify) Bone - Bone: base_scarf_1"!(ComponentPose=@n20, Alpha=1, Rotation="9.332,2.466,-20") -> (Pose) {
    rotationMode=BMM_Additive
    rotationSpace=BCS_BoneSpace
    translationMode=BMM_Additive
    translationSpace=BCS_BoneSpace
    alphaBoolBlend {blendInTime=0.5, blendOption=Linear, blendOutTime=0.5}
    boneToModify {boneName=base_scarf_1}
    rotation=<rot 9.332 2.466 -20>
   }
   n22 = "Transform (Modify) Bone - Bone: base_scarf_1"(ComponentPose=@n21, Alpha=1, Rotation="5,0,0") {
    rotationMode=BMM_Additive
    rotationSpace=BCS_BoneSpace
    translationMode=BMM_Additive
    translationSpace=BCS_BoneSpace
    alphaBoolBlend {blendInTime=0.5, blendOption=Linear, blendOutTime=0.5}
    alphaScaleBiasClamp {bInterpResult=true, interpSpeedDecreasing=8, interpSpeedIncreasing=8}
    boneToModify {boneName=base_scarf_1}
    rotation {pitch=5}
   }
   n23 = "Transform (Modify) Bone - Bone: base_scarf_1"(ComponentPose=@n22, Alpha=1, Rotation="12.195,0.827,15.562") {
    rotationMode=BMM_Additive
    rotationSpace=BCS_BoneSpace
    translationSpace=BCS_BoneSpace
    alphaBoolBlend {blendOption=Linear}
    alphaScaleBiasClamp {bInterpResult=true}
    boneToModify {boneName=base_scarf_1}
    rotation=<rot 12.195 0.827 15.562>
   }
   n24 = "Transform (Modify) Bone - Bone: base_radio"!(ComponentPose=@n23, bAlphaBoolEnabled:bool=True, Translation="0,0,2") -> (Pose) {
    alphaInputType=Bool
    rotationSpace=BCS_BoneSpace
    translationMode=BMM_Additive
    translationSpace=BCS_BoneSpace
    alphaBoolBlend {blendInTime=1, blendOption=Linear, blendOutTime=1}
    boneToModify {boneName=base_radio}
    translation {z=2}
   }
   n25 = "Transform (Modify) Bone - Bone: base_strap_back_0"!(ComponentPose=@n24, bAlphaBoolEnabled=True, Rotation="10,0,0") -> (Pose) {
    alphaInputType=Bool
    rotationMode=BMM_Additive
    rotationSpace=BCS_BoneSpace
    translationSpace=BCS_BoneSpace
    alphaBoolBlend {blendOption=Linear}
    boneToModify {boneName=base_strap_back_0}
    rotation {pitch=10}
   }
   n26 = "Transform (Modify) Bone - Bone: base_strap_back_0"(ComponentPose=@n25, bAlphaBoolEnabled=True, Rotation="7.385,-50.856,2.662") {
    alphaInputType=Bool
    rotationMode=BMM_Additive
    rotationSpace=BCS_BoneSpace
    translationSpace=BCS_BoneSpace
    alphaBoolBlend {blendOption=Linear}
    boneToModify {boneName=base_strap_back_0}
    rotation=<rot 7.385 -50.856 2.662>
   }
   n27 = "Transform (Modify) Bone - Bone: base_strap_front_0"(ComponentPose=@n26, bAlphaBoolEnabled=True, Translation="0,0,0", Rotation="31.112,28.827,32.336") {
    alphaInputType=Bool
    rotationMode=BMM_Additive
    rotationSpace=BCS_BoneSpace
    translationMode=BMM_Additive
    translationSpace=BCS_BoneSpace
    alphaBoolBlend {blendOption=Linear}
    boneToModify {boneName=base_strap_front_0}
    rotation=<rot 31.112 28.827 32.336>
   }
   Save cached pose 'modifyTransformCache'(Pose=Component To Local(ComponentPose=@n27)) {cacheName=modifyTransformCache}
   n28 = RigidBody!(
    ComponentPose=Local To Component(LocalPose=Use cached pose 'animDynCache'())
    Alpha
    OverrideWorldGravity:Vector="0,0,0"
    SimSpaceSettings:Sim Space Settings Structure=(WorldAlpha=0.4,VelocityScaleZ=1,DampingAlpha=0.8,MaxLinearVelocity=10000,MaxAngularVelocity=10000,MaxLinearAcceleration=10000,MaxAngularAcceleration=10000,ExternalLinearDragV=(X=0,Y=0,Z=0),ExternalLinearVelocity=(X=0,Y=0,Z=0),ExternalAngularVelocity=(X=0,Y=0,Z=0))
   ) -> (Pose) {
    bClampLinearTranslationLimitToRefPose=true
    bOverrideWorldGravity=true
    overridePhysicsAsset=&/F_MED_RaggedRebel_Physics
    simulationSpace=BaseBoneSpace
    alphaBoolBlend {blendOption=Linear}
    alphaScaleBiasClamp {bInterpResult=true}
    baseBoneRef {boneName=pelvis}
    simSpaceSettings {dampingAlpha=0.8, worldAlpha=0.4}
   }
   n29 = RigidBody!(ComponentPose=@n28, Alpha, OverrideWorldGravity="0,0,0") -> (Pose) {
    bClampLinearTranslationLimitToRefPose=true
    bOverrideWorldGravity=true
    overridePhysicsAsset=&/F_MED_RaggedRebel_Physics_Scarf
    simulationSpace=BaseBoneSpace
    alphaBoolBlend {blendOption=Linear}
    baseBoneRef {boneName=spine_05}
    simSpaceSettings {worldAlpha=0.45, externalLinearDragV {y=0.5}}
   }
   n30 = "Transform (Modify) Bone - Bone: dyn_scarf_3"!(ComponentPose=@n29, bAlphaBoolEnabled:bool=True, Rotation="-5,0,0") -> (Pose) {
    alphaInputType=Bool
    rotationMode=BMM_Additive
    rotationSpace=BCS_BoneSpace
    translationMode=BMM_Additive
    translationSpace=BCS_BoneSpace
    alphaBoolBlend {blendInTime=0.5, blendOption=Linear, blendOutTime=0.5}
    boneToModify {boneName=dyn_scarf_3}
    rotation {pitch=-5}
   }
   n31 = "Transform (Modify) Bone - Bone: dyn_scarf_4"!(ComponentPose=@n30, bAlphaBoolEnabled:bool=True, Rotation="-10,0,0") -> (Pose) {
    alphaInputType=Bool
    rotationMode=BMM_Additive
    rotationSpace=BCS_BoneSpace
    translationMode=BMM_Additive
    translationSpace=BCS_BoneSpace
    alphaBoolBlend {blendInTime=0.5, blendOption=Linear, blendOutTime=0.5}
    boneToModify {boneName=dyn_scarf_4}
    rotation {pitch=-10}
   }
   n32 = "Transform (Modify) Bone - Bone: dyn_scarf_5"!(ComponentPose=@n31, bAlphaBoolEnabled:bool=True, Rotation="10,0,0") -> (Pose) {
    alphaInputType=Bool
    rotationMode=BMM_Additive
    rotationSpace=BCS_BoneSpace
    translationMode=BMM_Additive
    translationSpace=BCS_BoneSpace
    alphaBoolBlend {blendInTime=0.5, blendOption=Linear, blendOutTime=0.5}
    boneToModify {boneName=dyn_scarf_5}
    rotation {pitch=10}
   }
   n33 = "Transform (Modify) Bone - Bone: dyn_strap_back_2"!(ComponentPose=@n32, bAlphaBoolEnabled:bool=True, Translation="2,0,2.7", Rotation="0,0,0") -> (Pose) {
    alphaInputType=Bool
    rotationMode=BMM_Additive
    rotationSpace=BCS_BoneSpace
    translationMode=BMM_Additive
    translationSpace=BCS_BoneSpace
    alphaBoolBlend {blendOption=Linear}
    boneToModify {boneName=dyn_strap_back_2}
    translation {x=2, z=2.7}
   }
   n34 = "Transform (Modify) Bone - Bone: dyn_strap_front_0"!(ComponentPose=@n33, bAlphaBoolEnabled:bool=True, Translation="0,0,0", Rotation="0.037,-16.610,-15.433") -> (Pose) {
    alphaInputType=Bool
    rotationMode=BMM_Additive
    rotationSpace=BCS_BoneSpace
    translationMode=BMM_Additive
    translationSpace=BCS_BoneSpace
    alphaBoolBlend {blendOption=Linear}
    boneToModify {boneName=dyn_strap_front_0}
    rotation=<rot 0.037 -16.61 -15.433>
   }
   Save cached pose 'RigidBodyCache'(Pose=Component To Local(ComponentPose=@n34)) {cacheName=RigidBodyCache}
   n35 = "Constraint - rivet_for_aim_grenade_1 (PPPP)"!(
    ComponentPose=Local To Component(LocalPose=Use cached pose 'RigidBodyCache'())
    Alpha=1
    ConstraintWeights_0=0.4817
    ConstraintWeights_1:float=0.4152
    ConstraintWeights_2:float=0.0816
    ConstraintWeights_3:float=0.0215
   ) -> (Pose) {
    alphaBoolBlend {blendOption=Linear}
    boneToModify {boneName=rivet_for_aim_grenade_1}
    constraintSetup [
     {targetBone {boneName=dyn_strap_front_3}}
     {targetBone {boneName=dyn_strap_front_4}}
     {targetBone {boneName=dyn_strap_front_5}}
     {targetBone {boneName=dyn_strap_front_2}}
    ]
    constraintWeights [1,1,1,1]
   }
   n36 = "Constraint - rivet_for_aim_grenade_2 (PPP)"!(ComponentPose=@n35, Alpha=1, ConstraintWeights_0=0.4904, ConstraintWeights_1:float=0.4008, ConstraintWeights_2:float=0.1088) -> (Pose) {
    alphaBoolBlend {blendOption=Linear}
    boneToModify {boneName=rivet_for_aim_grenade_2}
    constraintSetup [{targetBone {boneName=dyn_strap_back_4}}, {targetBone {boneName=dyn_strap_back_3}}, {targetBone {boneName=dyn_strap_back_5}}]
    constraintWeights [1,1,1]
   }
   n37 = "Constraint - base_grenade_1 (P)"!(ComponentPose=@n36, Alpha=1, ConstraintWeights_0=1, ConstraintWeights_1:float=0) -> (Pose) {
    alphaBoolBlend {blendOption=Linear}
    boneToModify {boneName=base_grenade_1}
    constraintSetup [{targetBone {boneName=dyn_strap_front_2}}]
    constraintWeights [1,0]
   }
   n38 = "Constraint - base_grenade_2 (P)"!(ComponentPose=@n37, Alpha=1, ConstraintWeights_0=1, ConstraintWeights_1:float=0) -> (Pose) {
    alphaBoolBlend {blendOption=Linear}
    boneToModify {boneName=base_grenade_2}
    constraintSetup [{targetBone {boneName=dyn_strap_back_2}}]
    constraintWeights [1,0]
   }
   n39 = Look At - Bone: base_grenade_1(ComponentPose=@n38, Alpha=1) {
    bUseLookUpAxis=true
    alphaBoolBlend {blendOption=Linear}
    boneToModify {boneName=base_grenade_1}
    lookAtTarget {boneReference {boneName=rivet_for_aim_grenade_1}}
    lookAt_Axis {axis=<1 0>}
    lookUp_Axis {axis {x=0, z=1}}
   }
   n40 = Look At - Bone: base_grenade_2(ComponentPose=@n39, Alpha=1) {
    bUseLookUpAxis=true
    alphaBoolBlend {blendOption=Linear}
    boneToModify {boneName=base_grenade_2}
    lookAtTarget {boneReference {boneName=rivet_for_aim_grenade_2}}
    lookAt_Axis {axis=<1 0>}
    lookUp_Axis {axis {x=0, z=1}}
   }
   n41 = "Transform (Modify) Bone - Bone: dyn_grenade_1_offset"!(ComponentPose=@n40, bAlphaBoolEnabled:bool=True, Translation="0,0,2", Rotation="4,13,4") -> (Pose) {
    alphaInputType=Bool
    rotationMode=BMM_Additive
    rotationSpace=BCS_BoneSpace
    translationMode=BMM_Additive
    translationSpace=BCS_BoneSpace
    alphaBoolBlend {blendOption=Linear}
    boneToModify {boneName=dyn_grenade_1_offset}
    rotation=<rot 4 13 4>
    translation {z=2}
   }
   n42 = "Transform (Modify) Bone - Bone: base_grenade_1"!(ComponentPose=@n41, bAlphaBoolEnabled:bool=True, Rotation="-12.832,-33.083,17.124") -> (Pose) {
    alphaInputType=Bool
    rotationMode=BMM_Additive
    rotationSpace=BCS_BoneSpace
    translationMode=BMM_Additive
    translationSpace=BCS_BoneSpace
    alphaBoolBlend {blendOption=Linear}
    boneToModify {boneName=base_grenade_1}
    rotation=<rot -12.832 -33.083 17.124>
   }
   n43 = "Transform (Modify) Bone - Bone: base_grenade_2"!(ComponentPose=@n42, bAlphaBoolEnabled:bool=True, Translation="0,1,0", Rotation="0,-32,15") -> (Pose) {
    alphaInputType=Bool
    rotationMode=BMM_Additive
    rotationSpace=BCS_BoneSpace
    translationSpace=BCS_BoneSpace
    alphaBoolBlend {blendOption=Linear}
    boneToModify {boneName=base_grenade_2}
    rotation {roll=15, yaw=-32}
    translation {y=1}
   }
   n44 = Anim Dynamics - Bone: dyn_grenade_1!(ComponentPose=@n43) -> (Pose) {
    angularDampingOverride=1
    angularSpringConstant=20000
    bAngularSpring=true
    bOverrideAngularDamping=true
    bUseGravityOverride=true
    gravityScale=0.35
    simulationSpace=RootRelative
    alphaBoolBlend {blendOption=Linear}
    boundBone {boneName=dyn_grenade_1}
    componentAppliedLinearAccClamp=<5000 5000 5000>
    componentLinearAccScale=<0.6 0.6 0.6>
    componentLinearVelScale=<0.6 0.6 0.6>
    gravityOverride {z=-980}
    physicsBodyDefinitions [{
      boundBone {boneName=dyn_grenade_1}
      boxExtents=<35 35 35>
      constraintSetup {angularLimitsMax=<0 0 55>, angularLimitsMin=<0 -40 -55>, angularTarget=<1 0 0>}
      localJointOffset=<8 0 0>
     }]
   }
   n45 = Anim Dynamics - Bone: dyn_grenade_2!(ComponentPose=@n44) -> (Pose) {
    angularDampingOverride=1
    angularSpringConstant=20000
    bAngularSpring=true
    bOverrideAngularDamping=true
    bUseGravityOverride=true
    gravityScale=0.35
    simulationSpace=RootRelative
    alphaBoolBlend {blendOption=Linear}
    boundBone {boneName=dyn_grenade_2}
    componentAppliedLinearAccClamp=<5000 5000 5000>
    componentLinearAccScale=<0.6 0.6 0.6>
    componentLinearVelScale=<0.6 0.6 0.6>
    gravityOverride {z=-980}
    physicsBodyDefinitions [{
      boundBone {boneName=dyn_grenade_2}
      boxExtents=<35 35 35>
      constraintSetup {angularLimitsMax=<0 0 40>, angularLimitsMin=<0 -60 -40>, angularTarget=<1 0 0>}
      localJointOffset=<8 0 0>
     }]
   }
   n46 = "Transform (Modify) Bone - Bone: base_grenade_1"(ComponentPose=@n45, bAlphaBoolEnabled=True, Rotation="4.834,-0.001,0.001") {
    alphaInputType=Bool
    rotationMode=BMM_Additive
    rotationSpace=BCS_BoneSpace
    translationMode=BMM_Additive
    translationSpace=BCS_BoneSpace
    alphaBoolBlend {blendOption=Linear}
    boneToModify {boneName=base_grenade_1}
    rotation=<rot 4.834 -0.001 0.001>
   }
   n47 = "Transform (Modify) Bone - Bone: base_grenade_2"(ComponentPose=@n46, bAlphaBoolEnabled=True, Translation="-0.033,0.543,0.284", Rotation="-7.181,31.697,-0.023") {
    alphaInputType=Bool
    rotationMode=BMM_Additive
    rotationSpace=BCS_BoneSpace
    translationMode=BMM_Additive
    translationSpace=BCS_BoneSpace
    alphaBoolBlend {blendOption=Linear}
    boneToModify {boneName=base_grenade_2}
    rotation=<rot -7.181 31.697 -0.023>
    translation=<-0.033 0.543 0.284>
   }
   n48 = "Transform (Modify) Bone - Bone: base_grenade_1"(ComponentPose=@n47, bAlphaBoolEnabled=True, Rotation="30,0,0") {
    alphaInputType=Bool
    rotationMode=BMM_Additive
    rotationSpace=BCS_BoneSpace
    translationMode=BMM_Additive
    translationSpace=BCS_BoneSpace
    alphaBoolBlend {blendOption=Linear}
    boneToModify {boneName=base_grenade_1}
    rotation {pitch=30}
   }
   n49 = "Transform (Modify) Bone - Bone: base_grenade_2"(ComponentPose=@n48, bAlphaBoolEnabled=True, Translation="0,0,0", Rotation="0.029,33.260,-0.015") {
    alphaInputType=Bool
    rotationMode=BMM_Additive
    rotationSpace=BCS_BoneSpace
    translationMode=BMM_Additive
    translationSpace=BCS_BoneSpace
    alphaBoolBlend {blendOption=Linear}
    boneToModify {boneName=base_grenade_2}
    rotation=<rot 0.029 33.26 -0.015>
   }
   Save cached pose 'GrenadesDynCache'(Pose=Component To Local(ComponentPose=@n49)) {cacheName=GrenadesDynCache}
   n50 = "Constraint - dyn_scarf_1 (PPPPPPPP)"!(
    ComponentPose=Local To Component(LocalPose=Use cached pose 'GrenadesDynCache'())
    bAlphaBoolEnabled:bool=True
    ConstraintWeights_0=0.5072
    ConstraintWeights_1:float=0.2507
    ConstraintWeights_2:float=0.1474
    ConstraintWeights_3:float=0.0846
    ConstraintWeights_4:float=0.0083
    ConstraintWeights_5:float=0.0009
    ConstraintWeights_6:float=0.0005
    ConstraintWeights_7:float=0.0003
   ) -> (Pose) {
    alphaInputType=Bool
    alphaBoolBlend {blendOption=Linear}
    boneToModify {boneName=dyn_scarf_1}
    constraintSetup [
     {targetBone {boneName=head}}
     {targetBone {boneName=neck_02}}
     {targetBone {boneName=spine_05}}
     {targetBone {boneName=neck_01}}
     {targetBone {boneName=spine_04}}
     {targetBone {boneName=clavicle_r}}
     {targetBone {boneName=spine_03}}
     {targetBone {boneName=clavicle_l}}
    ]
    constraintWeights [1,1,1,1,1,1,1,1]
   }
   n51 = "Constraint - dyn_scarf_2 (PPPPPPP)"!(
    ComponentPose=@n50
    bAlphaBoolEnabled:bool=True
    ConstraintWeights_0=0.6898
    ConstraintWeights_1:float=0.0989
    ConstraintWeights_2:float=0.0901
    ConstraintWeights_3:float=0.0684
    ConstraintWeights_4:float=0.0443
    ConstraintWeights_5:float=0.0076
    ConstraintWeights_6:float=0.001
   ) -> (Pose) {
    alphaInputType=Bool
    alphaBoolBlend {blendOption=Linear}
    boneToModify {boneName=dyn_scarf_2}
    constraintSetup [
     {targetBone {boneName=spine_05}}
     {targetBone {boneName=clavicle_r}}
     {targetBone {boneName=neck_01}}
     {targetBone {boneName=neck_02}}
     {targetBone {boneName=spine_03}}
     {targetBone {boneName=head}}
     {targetBone {boneName=clavicle_l}}
    ]
    constraintWeights [1,1,1,1,1,1,1]
   }
   n52 = "Constraint - dyn_scarf_3 (PPPPPPP)"!(
    ComponentPose=@n51
    bAlphaBoolEnabled:bool=True
    ConstraintWeights_0=0.6336
    ConstraintWeights_1:float=0.2846
    ConstraintWeights_2:float=0.0288
    ConstraintWeights_3:float=0.0266
    ConstraintWeights_4:float=0.0193
    ConstraintWeights_5:float=0.0039
    ConstraintWeights_6:float=0.0033
   ) -> (Pose) {
    alphaInputType=Bool
    alphaBoolBlend {blendOption=Linear}
    boneToModify {boneName=dyn_scarf_3}
    constraintSetup [
     {targetBone {boneName=clavicle_r}}
     {targetBone {boneName=spine_05}}
     {targetBone {boneName=spine_04}}
     {targetBone {boneName=upperarm_twist_01_r}}
     {targetBone {boneName=deform_upperarm_bk_r}}
     {targetBone {boneName=clavicle_l}}
     {targetBone {boneName=deform_upperarm_up_r}}
    ]
    constraintWeights [1,1,1,1,1,1,1]
   }
   n53 = "Constraint - dyn_scarf_4 (PPPPPPPP)"!(
    ComponentPose=@n52
    bAlphaBoolEnabled:bool=True
    ConstraintWeights_0=0.4504
    ConstraintWeights_1:float=0.2421
    ConstraintWeights_2:float=0.1747
    ConstraintWeights_3:float=0.0836
    ConstraintWeights_4:float=0.0307
    ConstraintWeights_5:float=0.008
    ConstraintWeights_6:float=0.0053
    ConstraintWeights_7:float=0.0052
   ) -> (Pose) {
    alphaInputType=Bool
    alphaBoolBlend {blendOption=Linear}
    boneToModify {boneName=dyn_scarf_4}
    constraintSetup [
     {targetBone {boneName=clavicle_r}}
     {targetBone {boneName=spine_05}}
     {targetBone {boneName=spine_04}}
     {targetBone {boneName=upperarm_twist_01_r}}
     {targetBone {boneName=deform_upperarm_bk_r}}
     {targetBone {boneName=spine_03}}
     {targetBone {boneName=spine_02}}
     {targetBone {boneName=clavicle_l}}
    ]
    constraintWeights [1,1,1,1,1,1,1,1]
   }
   n54 = "Constraint - dyn_scarf_5 (PPPPPPPPP)"!(
    ComponentPose=@n53
    bAlphaBoolEnabled:bool=True
    ConstraintWeights_0=0.3203
    ConstraintWeights_1:float=0.2833
    ConstraintWeights_2:float=0.1973
    ConstraintWeights_3:float=0.1043
    ConstraintWeights_4:float=0.0334
    ConstraintWeights_5:float=0.0324
    ConstraintWeights_6:float=0.0221
    ConstraintWeights_7:float=0.0037
    ConstraintWeights_8:float=0.0032
   ) -> (Pose) {
    alphaInputType=Bool
    alphaBoolBlend {blendOption=Linear}
    boneToModify {boneName=dyn_scarf_5}
    constraintSetup [
     {targetBone {boneName=clavicle_r}}
     {targetBone {boneName=spine_04}}
     {targetBone {boneName=spine_05}}
     {targetBone {boneName=upperarm_twist_01_r}}
     {targetBone {boneName=spine_03}}
     {targetBone {boneName=spine_02}}
     {targetBone {boneName=deform_upperarm_bk_r}}
     {targetBone {boneName=spine_01}}
     {targetBone {boneName=clavicle_l}}
    ]
    constraintWeights [1,1,1,1,1,1,1,1,1]
   }
   n55 = "Constraint - dyn_scarf_6 (PPPPPPPP)"!(
    ComponentPose=@n54
    bAlphaBoolEnabled:bool=True
    ConstraintWeights_0=0.3599
    ConstraintWeights_1:float=0.2574
    ConstraintWeights_2:float=0.1402
    ConstraintWeights_3:float=0.0706
    ConstraintWeights_4:float=0.0679
    ConstraintWeights_5:float=0.0608
    ConstraintWeights_6:float=0.0248
    ConstraintWeights_7:float=0.0184
   ) -> (Pose) {
    alphaInputType=Bool
    alphaBoolBlend {blendOption=Linear}
    boneToModify {boneName=dyn_scarf_6}
    constraintSetup [
     {targetBone {boneName=spine_04}}
     {targetBone {boneName=spine_02}}
     {targetBone {boneName=spine_03}}
     {targetBone {boneName=clavicle_r}}
     {targetBone {boneName=spine_05}}
     {targetBone {boneName=spine_01}}
     {targetBone {boneName=pelvis}}
     {targetBone {boneName=upperarm_twist_01_r}}
    ]
    constraintWeights [1,1,1,1,1,1,1,1]
   }
   n56 = "Transform (Modify) Bone - Bone: dyn_scarf_4"(ComponentPose=@n55, bAlphaBoolEnabled=True, Rotation="-15,0,0") {
    alphaInputType=Bool
    rotationMode=BMM_Additive
    rotationSpace=BCS_BoneSpace
    alphaBoolBlend {blendOption=Linear}
    boneToModify {boneName=dyn_scarf_4}
    rotation {pitch=-15}
   }
   Save cached pose 'CapeAndBackpackCache'(Pose=Component To Local(ComponentPose=@n56)) {cacheName=CapeAndBackpackCache}
  ]}
 EventGraph {nodes [
   Event BlueprintUpdateAnimation()
   > n1 = Sequence!(execute) -> (then_0, then_1, then_2:Exec, then_3:Exec, then_4:Exec)
   > n2 = Gravity Override State Function DEPRECATED(
    CharacterPartAnimInstance=Self-Reference()
    GameTimeInSeconds=GetGameTimeInSeconds()
    GravityOverrideParams=Get Gravity Override Params()
    RBAN Params=Get RBAN Params()
   )
   > Set GravityOverride(GravityOverride=@n2)
   > Set Sim Space Settings(Sim Space Settings:Sim Space Settings Structure=Break RBANSetting Struct!(RBANSettingStruct:"RBANSetting Struct Structure (by ref)"=@n2.RBAN SettingsOut) -> (
     ComponentLinearAccScale_28_#1:Vector=~
     ComponentLinearVelScale_31_#2:Vector=~
     SimSpaceSettings_34_#3:Sim Space Settings Structure
    ) {is=title K2Node_BreakStruct}.SimSpaceSettings_34_#3) -> (Output_Get:Sim Space Settings Structure)
   > n3 = Set Skel Mesh(Skel Mesh=GetOwnerSkeletalMeshComponent())
   > Set head_up(head_up:double=GetAlphaFromReferenceTransformDelta(SkeletalMeshComponent=@n3, BoneName=head, BaseName=spine_05, InputRange=(X=0,Y=-14), TwistAxis="0,0,1")) -> (Output_Get:double)
   > n4 = PlayerMovement(AnimInstance=Self-Reference(), XDotProductCoefficient=100, YDotProductCoefficient=100, ZDotProductCoefficient=100)
   n5 = BreakVector(InVec=MakeVector(X=@n4, Y=@n4.YMovement))
   > n6 = Set Left(Left:bool="float < float"(A=@n5, B=-50)) -> (Output_Get:bool)
   > n7 = Set Right(Right:bool="float > float"(A=@n5, B=50)) -> (Output_Get:bool)
   n8 = NOT Boolean(A=Get bIsCrouching())
   n9 = Get bIsMoving2D()
   n10 = Get bIsOnGround() -> (bIsOnGround:bool)
   n11 = OR Boolean(A=@n6, B=@n7)
   n12 = Get bIsOnGround()
   > n13 = Branch(Condition=OR Boolean!(
     A=AND Boolean!(A=@n12, B=@n11, C:bool=@n8) -> (ReturnValue)
     B=Get bIsParachuteOpened()
     C:bool
     D:bool=Get bIsDBNO()
     E:bool=AND Boolean(A=@n8, B=@n9, C=@n10)
    ) -> (ReturnValue))
   > Set left and right(left and right:double=0.5) -> (Output_Get:double)
   Set left and right(execute=@n13.else, left and right=1)
   n14 = Get bIsMoving2D()
   Set Crouch IDLE(execute=@n1.then_1, Crouch IDLE:bool=AND Boolean!(A=NOT Boolean(A=@n14), B=Get bIsCrouching()) -> (ReturnValue)) -> (Output_Get:bool)
   > n15 = Branch(Condition=AND Boolean(A=Get bIsCrouching(), B=Get bIsTargeting()))
   > Set Crouch Sprinting(Crouch Sprinting:double=Get head_up() -> (head_up:double)) -> (Output_Get:double)
   Set Crouch Sprinting(execute=@n15.else, Crouch Sprinting=0)
   n16 = CheckForBackbling()
   n17 = NOT Boolean(A=@n16.bIsBackblingLong)
   n18 = NOT Boolean(A=@n16.bIsBackblingACape)
   n19 = OR Boolean!(A=@n16.bIsBackblingACape, B=@n16.bIsBackblingLong) -> (ReturnValue)
   Set isWearingBackbling(execute=@n1.then_3, isWearingBackbling=OR Boolean(A=AND Boolean!(A=@n16, B=@n18, C:bool=@n17) -> (ReturnValue), B=@n19))
   > Set Is Wearing Cape(Is Wearing Cape=@n19)
   n20 = Get bIsSliding()
   > Set isCrouching(isCrouching:bool=AND Boolean!(A=Get bIsCrouching(), B=NOT Boolean(A=@n20)) -> (ReturnValue)) -> (Output_Get:bool)
   n21 = Get bRigidBodyEnabled()
   > Set RBAN off(RBAN off:bool=AND Boolean(A=NOT Boolean(A=@n21), B=Get bIsCrouching())) -> (Output_Get:bool)
   > n22 = Branch(Condition=Get bIsInFrontEnd())
   > Set Simulation Space(Simulation Space:Byte=2) -> (Output_Get:Byte=0)
   > Set alpha animDyn(alpha animDyn:double=0.5) -> (Output_Get:double)
   > Set alpha animDyn tongue shoe(alpha animDyn tongue shoe:double=0.3) -> (Output_Get:double)
   Set Simulation Space(execute=@n22.else, Simulation Space=3) -> (Output_Get=0)
   > Set alpha animDyn(alpha animDyn=1)
   > Set alpha animDyn tongue shoe(alpha animDyn tongue shoe=1)
   n23 = Branch(execute=@n1.then_2, Condition=Get bIsCrouching())
   > Set crouching targeting down(crouching targeting down:double=GetAlphaFromReferenceTransformDelta(SkeletalMeshComponent=@n3, BoneName=head, BaseName=spine_05, InputRange=(X=0,Y=1), TwistAxis="0,0,1")) -> (Output_Get:double)
   Set crouching targeting down(execute=@n23.else, crouching targeting down=0)
   Get bIsPlayingEmote()
  ]}
]
variables [
 alpha animDyn {type=number}
 alpha animDyn tongue shoe {type=number}
 Crouch IDLE {type=boolean}
 Crouch Sprinting {type=number}
 crouching RBAN off {type=boolean}
 crouching targeting down {type=number}
 Gravity Override Params {type=GravityOverrideParamsStruct}
 GravityOverride {type=Vector}
 head_up {type=number}
 Is Wearing Cape {type=boolean}
 isCrouching {type=boolean}
 isWearingBackbling {type=boolean}
 Left {type=boolean}
 left and right {type=number}
 RBAN off {type=boolean}
 RBAN Params {
  type=RBANParamsStruct
  default {
   dBNO {simSpaceSettings {
     dampingAlpha=0.8
     maxAngularAcceleration=10000
     maxAngularVelocity=10000
     maxLinearAcceleration=10000
     maxLinearVelocity=10000
     velocityScaleZ=1
     worldAlpha=0.4
     externalLinearDragV=<0 0 0>
    }}
   dBNO_Carried {simSpaceSettings {
     dampingAlpha=0.8
     maxAngularAcceleration=10000
     maxAngularVelocity=10000
     maxLinearAcceleration=10000
     maxLinearVelocity=10000
     velocityScaleZ=1
     worldAlpha=0.4
     externalLinearDragV=<0 0 0>
    }}
   emoteOrMelee {simSpaceSettings {
     maxAngularAcceleration=10000
     maxAngularVelocity=10000
     maxLinearAcceleration=10000
     maxLinearVelocity=10000
     velocityScaleZ=1
     worldAlpha=0.8
     externalLinearDragV=<0 0 0>
    }}
   front_End {simSpaceSettings {
     maxAngularAcceleration=10000
     maxAngularVelocity=10000
     maxLinearAcceleration=10000
     maxLinearVelocity=10000
     velocityScaleZ=2
     worldAlpha=0.5
     externalLinearDragV=<0 0 0>
    }}
   onGround_Crouch_Moving {simSpaceSettings {
     dampingAlpha=0.8
     maxAngularAcceleration=10000
     maxAngularVelocity=10000
     maxLinearAcceleration=10000
     maxLinearVelocity=10000
     velocityScaleZ=1
     worldAlpha=0.4
     externalLinearDragV=<0 0 0>
    }}
   onGround_Falling {simSpaceSettings {
     dampingAlpha=0.8
     maxAngularAcceleration=10000
     maxAngularVelocity=10000
     maxLinearAcceleration=10000
     maxLinearVelocity=10000
     velocityScaleZ=1
     worldAlpha=0.4
     externalLinearDragV=<0 0 0>
    }}
   onGround_Moving {simSpaceSettings {
     dampingAlpha=0.8
     maxAngularAcceleration=10000
     maxAngularVelocity=10000
     maxLinearAcceleration=10000
     maxLinearVelocity=10000
     velocityScaleZ=1
     worldAlpha=0.4
     externalLinearDragV=<0 0 0>
    }}
   onGround_Standing {simSpaceSettings {
     dampingAlpha=0.8
     maxAngularAcceleration=10000
     maxAngularVelocity=10000
     maxLinearAcceleration=10000
     maxLinearVelocity=10000
     velocityScaleZ=1
     worldAlpha=0.4
     externalLinearDragV=<0 0 0>
    }}
   rail_Grinding {simSpaceSettings {
     dampingAlpha=0.8
     maxAngularAcceleration=10000
     maxAngularVelocity=10000
     maxLinearAcceleration=10000
     maxLinearVelocity=10000
     velocityScaleZ=1
     worldAlpha=0.4
     externalLinearDragV=<0 0 0>
    }}
   riding_Moving {simSpaceSettings {
     dampingAlpha=0.8
     maxAngularAcceleration=10000
     maxAngularVelocity=10000
     maxLinearAcceleration=10000
     maxLinearVelocity=10000
     velocityScaleZ=1
     worldAlpha=0.4
     externalLinearDragV=<0 0 0>
    }}
   riding_Moving_Targeting {simSpaceSettings {
     dampingAlpha=0.8
     maxAngularAcceleration=10000
     maxAngularVelocity=10000
     maxLinearAcceleration=10000
     maxLinearVelocity=10000
     velocityScaleZ=1
     worldAlpha=0.4
     externalLinearDragV=<0 0 0>
    }}
   riding_Standing {simSpaceSettings {
     dampingAlpha=0.8
     maxAngularAcceleration=10000
     maxAngularVelocity=10000
     maxLinearAcceleration=10000
     maxLinearVelocity=10000
     velocityScaleZ=1
     worldAlpha=0.4
     externalLinearDragV=<0 0 0>
    }}
   skydive_DivingDown {simSpaceSettings {
     dampingAlpha=0.8
     maxAngularAcceleration=10000
     maxAngularVelocity=10000
     maxLinearAcceleration=10000
     maxLinearVelocity=10000
     velocityScaleZ=1
     worldAlpha=0.4
     externalLinearDragV=<0 0 0>
    }}
   skydive_FloatingDown {
    componentLinearAccScale=<0 0.5 0>
    componentLinearVelScale=<0 0.5 0>
    simSpaceSettings {
     dampingAlpha=0.8
     maxAngularAcceleration=10000
     maxAngularVelocity=10000
     maxLinearAcceleration=10000
     maxLinearVelocity=10000
     velocityScaleZ=1
     worldAlpha=0.4
     externalLinearDragV=<0 0 0>
    }
   }
   skydive_GoingUp {simSpaceSettings {
     dampingAlpha=0.8
     maxAngularAcceleration=10000
     maxAngularVelocity=10000
     maxLinearAcceleration=10000
     maxLinearVelocity=10000
     velocityScaleZ=1
     worldAlpha=0.4
     externalLinearDragV=<0 0 0>
    }}
   skydive_Parachuting {
    componentLinearAccScale=<0 0.1 0.3>
    componentLinearVelScale=<0 0.1 0.3>
    simSpaceSettings {
     dampingAlpha=0.8
     maxAngularAcceleration=10000
     maxAngularVelocity=10000
     maxLinearAcceleration=10000
     maxLinearVelocity=10000
     velocityScaleZ=1
     worldAlpha=0.4
     externalLinearDragV=<0 0 0>
    }
   }
   skydive_Surfing {simSpaceSettings {
     dampingAlpha=0.8
     maxAngularAcceleration=10000
     maxAngularVelocity=10000
     maxLinearAcceleration=10000
     maxLinearVelocity=10000
     velocityScaleZ=1
     worldAlpha=0.4
     externalLinearDragV=<0 0 0>
    }}
   swimming {simSpaceSettings {
     dampingAlpha=0.8
     maxAngularAcceleration=10000
     maxAngularVelocity=10000
     maxLinearAcceleration=10000
     maxLinearVelocity=10000
     velocityScaleZ=1
     worldAlpha=0.4
     externalLinearDragV=<0 0 0>
    }}
   tactical_Sprint {simSpaceSettings {
     dampingAlpha=0.8
     maxAngularAcceleration=10000
     maxAngularVelocity=10000
     maxLinearAcceleration=10000
     maxLinearVelocity=10000
     velocityScaleZ=1
     worldAlpha=0.4
     externalLinearDragV=<0 0 0>
    }}
   ziplining {simSpaceSettings {
     dampingAlpha=0.8
     maxAngularAcceleration=10000
     maxAngularVelocity=10000
     maxLinearAcceleration=10000
     maxLinearVelocity=10000
     velocityScaleZ=1
     worldAlpha=0.4
     externalLinearDragV=<0 0 0>
    }}
  }
 }
 Right {type=boolean}
 Sim Space Settings {
  type=SimSpaceSettings
  default {
   dampingAlpha=1
   maxAngularAcceleration=10000
   maxAngularVelocity=10000
   maxLinearAcceleration=10000
   maxLinearVelocity=10000
   velocityScaleZ=1
   worldAlpha=0
   externalLinearDragV=<0 0 0>
  }
 }
 Simulation Space {type=integer}
 Skel Mesh {type=SkeletalMeshComponent}
]
