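# Physics asset for the face accessory part of the female medium RaggedRebel player cosmetic, covering jointed hair and a scarf.
# Capsules on the head, neck, upper spine, clavicles and upper arms form a body proxy linked by free-swinging constraints.
# Four hair chains (back, back-left, left, right) hang from head-attached base bones with tight twist and varied swing limits.
# A six-link scarf chain hangs from the neck, twist-locked after its free first link.
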PhysicsAsset /BRCosmetics/Characters/Player/Female/Medium/Bodies/F_MED_RaggedRebel/Meshes/Parts/F_MED_RaggedRebel_FaceAcc_Physics
bodies {
 base_hair_back [{shapeType=Capsule, length=1, radius=1, shapeName=base_hair_back_capsule}]
 base_hair_backlf [{shapeType=Capsule, length=1, radius=1, shapeName=base_hair_backlf_capsule}]
 base_hair_lf [{shapeType=Capsule, length=1, radius=1, shapeName=base_hair_lf_capsule}]
 base_hair_rt [{shapeType=Capsule, length=1, radius=1, shapeName=base_hair_rt_capsule}]
 base_scarf_1 [{shapeType=Capsule, length=0.505, radius=0.505, shapeName=base_scarf_1_capsule, rotation=<rot 0 -90 0>}]
 clavicle_l [{shapeType=Capsule, length=6, radius=9, shapeName=clavicle_l_capsule, center=<7.3 0 -3.2>, rotation=<rot 90 0 -10>}]
 clavicle_r [{shapeType=Capsule, length=6, radius=10, shapeName=clavicle_l_capsule, center=<-7.3 0 3.2018>, rotation=<rot 90 165.9638 155.9638>}]
 dyn_hair_back_01 [{shapeType=Capsule, length=3, radius=3, shapeName=dyn_hair_back_01_capsule, center=<3 0 0.5>, rotation=<rot 90 0 0>}]
 dyn_hair_back_02 [{shapeType=Capsule, length=6, radius=3, shapeName=dyn_hair_back_02_capsule, center=<2.5 0 0>, rotation=<rot 0 0 90>}]
 dyn_hair_back_03 [{shapeType=Capsule, length=6, radius=3, shapeName=dyn_hair_back_03_capsule, center=<2.5 0 0.8>, rotation=<rot 0 0 90>}]
 dyn_hair_back_04 [{shapeType=Capsule, length=2, radius=2, shapeName=dyn_hair_back_04_capsule, center=<1 0 0.5>, rotation=<rot 75 0 0>}]
 dyn_hair_back_05 [{shapeType=Capsule, length=5, radius=3, shapeName=dyn_hair_back_05_capsule, center=<3 0 -1>, rotation=<rot 90 0 0>}]
 dyn_hair_back_06 [{shapeType=Capsule, length=2, radius=3, shapeName=dyn_hair_back_06_capsule, center=<3 0 -0.5>, rotation=<rot 110 0 0>}]
 dyn_hair_back_07 [{shapeType=Capsule, length=2, radius=3, shapeName=dyn_hair_back_07_capsule, center=<2 0 0.5>, rotation=<rot 90 0 0>}]
 dyn_hair_backlf_01 [{shapeType=Capsule, length=2, radius=3, shapeName=dyn_hair_backlf_01_capsule, center=<3.5 0 0>, rotation=<rot 90 -15.6968 -15.6969>}]
 dyn_hair_backlf_02 [{shapeType=Capsule, length=7, radius=3, shapeName=dyn_hair_backlf_02_capsule, center=<4.5 0 0>, rotation=<rot 90 15.1118 15.1118>}]
 dyn_hair_lf_01 [{shapeType=Capsule, length=6, radius=1.5, shapeName=dyn_hair_lf_01_capsule, center=<4 2.5 0.5>, rotation=<rot 90 -47.8105 -60>}]
 dyn_hair_lf_02 [{shapeType=Capsule, length=11, radius=3, shapeName=dyn_hair_lf_02_capsule, center=<5 -1.3 -1>, rotation=<rot 85 0 40>}]
 dyn_hair_rt_01 [{shapeType=Capsule, length=9, radius=2, shapeName=dyn_hair_rt_01_capsule, center=<4 1 1.5>, rotation=<rot 90 0 -10>}]
 dyn_hair_rt_02 [{shapeType=Capsule, length=11, radius=2.9, shapeName=dyn_hair_rt_02_capsule, center=<3 1 -0.5>, rotation=<rot 90 180 150>}]
 dyn_scarf_1 [{shapeType=Capsule, length=8, radius=1, shapeName=dyn_scarf_1_capsule, center=<2 0 0>, rotation=<rot 90 0 0>}]
 dyn_scarf_2 [{shapeType=Capsule, length=8.6501, radius=1, shapeName=dyn_scarf_2_capsule, center=<2 0 0>, rotation=<rot 90 0 0>}]
 dyn_scarf_3 [{shapeType=Capsule, length=8.6501, radius=1, shapeName=dyn_scarf_3_capsule, center=<2 0 0>, rotation=<rot 90 0 0>}]
 dyn_scarf_4 [{shapeType=Capsule, length=8.6501, radius=1, shapeName=dyn_scarf_4_capsule, center=<2 0 0>, rotation=<rot 90 0 0>}]
 dyn_scarf_5 [{shapeType=Capsule, length=8.6501, radius=1, shapeName=dyn_scarf_5_capsule, center=<2 0 0>, rotation=<rot 90 0 0>}]
 dyn_scarf_6 [{shapeType=Capsule, length=8.6501, radius=1, shapeName=dyn_scarf_6_capsule, center=<2 0 0>, rotation=<rot 90 0 0>}]
 head [{shapeType=Capsule, length=11, radius=6.2, shapeName=head_capsule, center=<3.7 2 0>, rotation=<rot 0 60 90>}]
 neck_01 [{shapeType=Capsule, length=2, radius=6, shapeName=neck_01_capsule, center=<0 2 0>}]
 neck_02 [{shapeType=Capsule, length=33, radius=3, shapeName=neck_02_capsule, center=<0 2 0>}]
 spine_03 [{shapeType=Capsule, length=22, radius=11.2, shapeName=spine_04_capsule, center=<-12.5 0 0>, rotation=<rot 90 0 -15>}]
 spine_04 [{shapeType=Capsule, length=11, radius=9.8, shapeName=spine_04_capsule, center=<-4 0 0>}]
 spine_05 [{shapeType=Capsule, length=16, radius=11, shapeName=spine_05_capsule, center=<2 2 0>}]
 upperarm_l [{shapeType=Capsule, length=18, radius=6, shapeName=upperarm_l_capsule, center=<16 0 0.8>, rotation=<rot 90 0 0>}]
 upperarm_r [{shapeType=Capsule, length=18, radius=6, shapeName=upperarm_l_capsule, center=<-16 0 -0.8>, rotation=<rot 90 0 0>}]
}
constraints [
 {
  bone1Name=neck_02
  bone2Name=neck_01
  swing1LimitDegrees=45
  swing1Motion=Free
  swing2LimitDegrees=45
  swing2Motion=Free
  twistLimitDegrees=45
  twistMotion=Free
 }
 {
  bone1Name=head
  bone2Name=neck_02
  swing1LimitDegrees=45
  swing1Motion=Free
  swing2LimitDegrees=45
  swing2Motion=Free
  twistLimitDegrees=45
  twistMotion=Free
 }
 {
  bone1Name=base_hair_back
  bone2Name=head
  swing1LimitDegrees=45
  swing1Motion=Free
  swing2LimitDegrees=45
  swing2Motion=Free
  twistLimitDegrees=45
  twistMotion=Free
 }
 {bone1Name=dyn_hair_back_01, bone2Name=base_hair_back, swing1LimitDegrees=1, swing2LimitDegrees=20, twistLimitDegrees=1}
 {bone1Name=dyn_hair_back_02, bone2Name=dyn_hair_back_01, swing1LimitDegrees=3, swing2LimitDegrees=30, twistLimitDegrees=5}
 {bone1Name=dyn_hair_back_03, bone2Name=dyn_hair_back_02, swing1LimitDegrees=5, swing2LimitDegrees=35, twistLimitDegrees=10}
 {bone1Name=dyn_hair_back_04, bone2Name=dyn_hair_back_03, swing1LimitDegrees=10, swing2LimitDegrees=45, twistLimitDegrees=5}
 {bone1Name=dyn_hair_back_05, bone2Name=dyn_hair_back_04, swing1LimitDegrees=5, swing2LimitDegrees=30, twistLimitDegrees=1}
 {bone1Name=dyn_hair_back_06, bone2Name=dyn_hair_back_05, swing1LimitDegrees=5, swing2LimitDegrees=15, twistLimitDegrees=1}
 {bone1Name=dyn_hair_back_07, bone2Name=dyn_hair_back_06, swing1LimitDegrees=5, swing2LimitDegrees=10, twistLimitDegrees=1}
 {
  bone1Name=base_hair_backlf
  bone2Name=head
  swing1LimitDegrees=45
  swing1Motion=Free
  swing2LimitDegrees=45
  swing2Motion=Free
  twistLimitDegrees=45
  twistMotion=Free
 }
 {bone1Name=dyn_hair_backlf_01, bone2Name=base_hair_backlf, swing1LimitDegrees=15, swing2LimitDegrees=15, twistLimitDegrees=1}
 {bone1Name=dyn_hair_backlf_02, bone2Name=dyn_hair_backlf_01, swing1LimitDegrees=20, swing2LimitDegrees=30, twistLimitDegrees=1}
 {
  bone1Name=base_hair_lf
  bone2Name=head
  swing1LimitDegrees=45
  swing1Motion=Free
  swing2LimitDegrees=45
  swing2Motion=Free
  twistLimitDegrees=45
  twistMotion=Free
 }
 {bone1Name=dyn_hair_lf_01, bone2Name=base_hair_lf, swing1LimitDegrees=15, swing2LimitDegrees=15, twistLimitDegrees=1}
 {bone1Name=dyn_hair_lf_02, bone2Name=dyn_hair_lf_01, swing1LimitDegrees=30, swing2LimitDegrees=20, twistLimitDegrees=1}
 {
  bone1Name=base_hair_rt
  bone2Name=head
  swing1LimitDegrees=45
  swing1Motion=Free
  swing2LimitDegrees=45
  swing2Motion=Free
  twistLimitDegrees=45
  twistMotion=Free
 }
 {bone1Name=dyn_hair_rt_01, bone2Name=base_hair_rt, swing1LimitDegrees=15, swing2LimitDegrees=15, twistLimitDegrees=1}
 {bone1Name=dyn_hair_rt_02, bone2Name=dyn_hair_rt_01, swing1LimitDegrees=30, swing2LimitDegrees=30, twistLimitDegrees=1}
 {
  bone1Name=neck_01
  bone2Name=spine_05
  swing1LimitDegrees=45
  swing1Motion=Free
  swing2LimitDegrees=45
  swing2Motion=Free
  twistLimitDegrees=45
  twistMotion=Free
 }
 {
  bone1Name=spine_05
  bone2Name=spine_04
  swing1LimitDegrees=45
  swing1Motion=Free
  swing2LimitDegrees=45
  swing2Motion=Free
  twistLimitDegrees=45
  twistMotion=Free
 }
 {
  bone1Name=spine_04
  bone2Name=spine_03
  swing1LimitDegrees=45
  swing1Motion=Free
  swing2LimitDegrees=45
  swing2Motion=Free
  twistLimitDegrees=45
  twistMotion=Free
 }
 {
  bone1Name=clavicle_l
  bone2Name=spine_05
  swing1LimitDegrees=45
  swing1Motion=Free
  swing2LimitDegrees=45
  swing2Motion=Free
  twistLimitDegrees=45
  twistMotion=Free
 }
 {
  bone1Name=upperarm_l
  bone2Name=clavicle_l
  swing1LimitDegrees=45
  swing1Motion=Free
  swing2LimitDegrees=45
  swing2Motion=Free
  twistLimitDegrees=45
  twistMotion=Free
 }
 {
  bone1Name=clavicle_r
  bone2Name=spine_05
  swing1LimitDegrees=45
  swing1Motion=Free
  swing2LimitDegrees=45
  swing2Motion=Free
  twistLimitDegrees=45
  twistMotion=Free
 }
 {
  bone1Name=upperarm_r
  bone2Name=clavicle_r
  swing1LimitDegrees=45
  swing1Motion=Free
  swing2LimitDegrees=45
  swing2Motion=Free
  twistLimitDegrees=45
  twistMotion=Free
 }
 {
  bone1Name=base_scarf_1
  bone2Name=neck_02
  swing1LimitDegrees=45
  swing1Motion=Free
  swing2LimitDegrees=45
  swing2Motion=Free
  twistLimitDegrees=45
  twistMotion=Free
 }
 {
  bone1Name=dyn_scarf_1
  bone2Name=base_scarf_1
  swing1LimitDegrees=45
  swing1Motion=Free
  swing2LimitDegrees=45
  swing2Motion=Free
  twistLimitDegrees=45
  twistMotion=Free
 }
 {bone1Name=dyn_scarf_2, bone2Name=dyn_scarf_1, swing1LimitDegrees=30, swing2LimitDegrees=30, twistLimitDegrees=0}
 {bone1Name=dyn_scarf_3, bone2Name=dyn_scarf_2, swing1LimitDegrees=30, swing2LimitDegrees=30, twistLimitDegrees=0}
 {bone1Name=dyn_scarf_4, bone2Name=dyn_scarf_3, swing1LimitDegrees=30, swing2LimitDegrees=30, twistLimitDegrees=0}
 {bone1Name=dyn_scarf_5, bone2Name=dyn_scarf_4, swing1LimitDegrees=30, swing2LimitDegrees=30, twistLimitDegrees=0}
 {bone1Name=dyn_scarf_6, bone2Name=dyn_scarf_5, swing1LimitDegrees=30, swing2LimitDegrees=30, twistLimitDegrees=0}
]
